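# A stereo MetaSound Source preset that plays the looping Stage 01 idle music for the Gifted Traveler.
# It inherits its graph from MSS_GiftedTraveler_BasicLoop_Stereo and supplies the music wave and a master volume as inputs.
# The sound uses the Gifted Traveler music attenuation and the NPC music sound class.
# It is flagged as looping and virtualizes as PlayWhenSilent, so it keeps playing while inaudible.
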MetaSoundSource /GiftedTraveler/Assets/Audio/MSS_GiftedTraveler_IdleMusicLoop_Stereo
preset_parent=&/MSS_GiftedTraveler_BasicLoop_Stereo
inputs [
 Master Volume {type=Float}
 UE.Source.OnPlay {type=Trigger}
 UE.Source.StartTime.StartTime {type=Time}
 Wave Asset {type=WaveAsset}
]
interfaces [UE.OutputFormat.Stereo:1.0, UE.Source.StartTime:1.0, UE.Source:1.1]
outputs [
 "UE.OutputFormat.Stereo.Audio:0" {type=Audio}
 "UE.OutputFormat.Stereo.Audio:1" {type=Audio}
]
pages [Default {input_defaults {
   Master Volume=3
   UE.Source.OnPlay=null
   UE.Source.StartTime.StartTime=null
   Wave Asset=/GiftedTraveler/Assets/Audio/Music/MX_S34_GiftedTraveler_Stage_01_LP
  }}]
properties {
 assetImportData=&self:AssetImportData
 attenuationSettings=/GiftedTraveler/Assets/Audio/Music/ATT_MX_GiftedTraveler
 bLooping=true
 duration=10000
 maxDistance=8500
 numChannels=2
 outputFormat=Stereo
 sampleRate=48000
 soundClassObject=&Fort_Music_NPC
 virtualizationMode=PlayWhenSilent
}
