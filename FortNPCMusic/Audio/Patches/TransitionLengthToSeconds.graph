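# MetaSound patch for NPC music that converts an integer transition length into seconds, output as a float.
# BPM, beat multiplier and divisions of a whole note give the duration of one division via BPMToSeconds.
# That duration is scaled by transition length times divisions, so the length effectively counts whole notes at the default multiplier.
# Divisions are converted to an integer for that product, so fractional values are not preserved there.
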MetaSoundPatch /FortNPCMusic/Audio/Patches/TransitionLengthToSeconds
inputs [
 Beat Multiplier {type=Float}
 BPM {type=Float}
 DivisionsOfWholeNote {type=Float}
 TransitionLength {type=Int32}
]
outputs [Output {type=Float}]
pages [Default {
  input_defaults {BPM=0, Beat Multiplier=1, DivisionsOfWholeNote=4, TransitionLength=0}
  nodes [
   n1 = UE.Multiply:Int32(PrimaryOperand=^TransitionLength, AdditionalOperands=Convert.Int32:Float(Float=^DivisionsOfWholeNote).Int32) {also {version [1,1]}}
   ConversionTimeToFloat_0 = UE.ConversionTimeToFloat(In=UE.Multiply:Time by Float(
     PrimaryOperand=UE.BPMToSeconds(BPM=^BPM, Beat Multiplier=^Beat Multiplier, Divisions of Whole Note=^DivisionsOfWholeNote).Seconds
     AdditionalOperands=Convert.Float:Int32(Int32=@n1.Out).Float
    ).Out)
  ]
  links [{dst="@Output", src=ConversionTimeToFloat_0.Out}]
 }]
