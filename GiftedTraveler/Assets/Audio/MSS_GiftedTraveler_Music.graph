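# MetaSound source for Gifted Traveler's interactive NPC music, moving between three stage loops, transition stingers and three endings on game trigger inputs.
# A metronome and beat counter produce a bar pulse, which ScheduleTrigger patches use to quantize events, delaying stage loops by the transition length after stingers.
# Stages 2 and 3 layer percussion and other stems, whose volumes and fades change during transitions.
# All players feed one stereo mixer scaled by master volume.
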MetaSoundSource /GiftedTraveler/Assets/Audio/MSS_GiftedTraveler_Music
inputs [
 AdditionalOperands {type=Float}
 Beats per Bar {type=Int32}
 BPM {type=Float}
 EndingDelay {type=Int32}
 Loop Length Bars {type=Int32}
 Loop.Stage01Loop {type=WaveAsset}
 Loop.Stage02Else {type=WaveAsset}
 Loop.Stage02Perc {type=WaveAsset}
 Loop.Stage03Else {type=WaveAsset}
 Loop.Stage03Perc {type=WaveAsset}
 MasterVolume {type=Float}
 NPC.Event.PlayEnding01 {type=Trigger}
 NPC.Event.PlayEnding02 {type=Trigger}
 NPC.Event.PlayEnding03 {type=Trigger}
 NPC.Event.PlayStage1Loop {type=Trigger}
 NPC.Event.PlayStage2Loop {type=Trigger}
 NPC.Event.PlayStage3Loop {type=Trigger}
 NPC.Event.PlayTransitionToEnding {type=Trigger}
 NPC.Event.PlayTransitionToStage1 {type=Trigger}
 NPC.Event.PlayTransitionToStage2 {type=Trigger}
 NPC.Event.PlayTransitionToStage3 {type=Trigger}
 NPC.Event.StopAll {type=Trigger}
 Stinger.Ending01 {type=WaveAsset}
 Stinger.Ending02 {type=WaveAsset}
 Stinger.Ending03 {type=WaveAsset}
 Stinger.TransitionToStage02 {type=WaveAsset}
 Stinger.TransitionToStage03 {type=WaveAsset}
 TransitionLengthsBars {type=Int32:Array}
 UE.Source.OnPlay {type=Trigger}
]
interfaces [UE.OutputFormat.Stereo:1.0, UE.Source:1.1]
outputs [
 CurrentPhase {type=Int32}
 "UE.OutputFormat.Stereo.Audio:0" {type=Audio}
 "UE.OutputFormat.Stereo.Audio:1" {type=Audio}
]
pages [Default {
  input_defaults {
   AdditionalOperands=2.06
   BPM=150
   Beats per Bar=4
   EndingDelay=1
   Loop Length Bars=16
   Loop.Stage01Loop=null
   Loop.Stage02Else=null
   Loop.Stage02Perc=null
   Loop.Stage03Else=null
   Loop.Stage03Perc=null
   MasterVolume=0.7079
   NPC.Event.PlayEnding01=null
   NPC.Event.PlayEnding02=null
   NPC.Event.PlayEnding03=null
   NPC.Event.PlayStage1Loop=null
   NPC.Event.PlayStage2Loop=null
   NPC.Event.PlayStage3Loop=null
   NPC.Event.PlayTransitionToEnding=null
   NPC.Event.PlayTransitionToStage1=null
   NPC.Event.PlayTransitionToStage2=null
   NPC.Event.PlayTransitionToStage3=null
   NPC.Event.StopAll=null
   Stinger.Ending01=null
   Stinger.Ending02=null
   Stinger.Ending03=null
   Stinger.TransitionToStage02=null
   Stinger.TransitionToStage03=null
   UE.Source.OnPlay=null
   TransitionLengthsBars [4,4]
  }
  nodes [
   TriggerCounter_0 = UE.Trigger Counter(In=/FortNPCMusic/Audio/Patches/MSP_NPCMusic_Metronome(BPM=^BPM, UE.Source.OnPlay=^UE.Source.OnPlay).OnBeat) {also {defaults {Reset Count=4}}}
   n1 = UE.Modulo:Int32(PrimaryOperand=@TriggerCounter_0.Count) {also {defaults {AdditionalOperands=4}}}
   n2 = VariableAccessor.Trigger(Variable=VariableMutator.Trigger(
     Variable=InitVariable.Trigger().Variable
     Value=TriggerCompare.Clamp:Int32(A=@n1.Out, Compare=@TriggerCounter_0.On Trigger) {also {defaults {B=1}}}.True
    ).Variable)
   n3 = VariableAccessor.Trigger(Variable=@n2.Variable)
   n4 = VariableAccessor.Trigger(Variable=@n3.Variable)
   n5 = VariableAccessor.Trigger(Variable=@n4.Variable)
   n6 = VariableAccessor.Trigger(Variable=@n5.Variable)
   n7 = VariableAccessor.Trigger(Variable=@n6.Variable)
   n8 = &MSP_ScheduleTrigger(Schedule=^NPC.Event.PlayTransitionToStage3, In Trigger=@n7.Value)
   n9 = VariableAccessor.Trigger(Variable=VariableMutator.Trigger(Variable=InitVariable.Trigger().Variable, Value=@n8.Scheduled Trigger).Variable)
   TriggerAnyTriggerAny2_1 = "TriggerAny.Trigger Any (2)"(In 0=^NPC.Event.PlayStage3Loop, In 1=@n9.Value)
   n10 = VariableDeferredAccessor.Trigger(Variable=InitVariable.Trigger().Variable)
   n11 = VariableDeferredAccessor.Trigger(Variable=InitVariable.Trigger().Variable)
   n12 = VariableDeferredAccessor.Trigger(Variable=InitVariable.Trigger().Variable)
   n13 = VariableDeferredAccessor.Trigger(Variable=InitVariable.Trigger().Variable)
   TriggerRouteTriggerRouteInt325_0 = "TriggerRoute.Trigger Route (Int32, 5):Int32"(Set 2=@n10.Value, Set 1=@n11.Value, Set 4=@n12.Value, Set 3=@n13.Value) {also {defaults {Value 1=1, Value 2=2, Value 3=3, Value 4=4}}}
   n14 = VariableAccessor.Int32(Variable=VariableMutator.Int32(Variable=InitVariable.Int32().Variable, Value=@TriggerRouteTriggerRouteInt325_0.Value).Variable)
   n15 = VariableDeferredAccessor.Trigger(Variable=InitVariable.Trigger().Variable)
   n16 = VariableAccessor.Int32(Variable=VariableMutator.Int32(
     Value=Array.Get:Int32:Array(Array=^TransitionLengthsBars, Index=@n14.Value, Trigger=@n15.Value).Element
     Variable=InitVariable.Int32().Variable
    ).Variable)
   n17 = VariableAccessor.Int32(Variable=@n16.Variable)
   TriggerRouteTriggerRouteInt322_0 = "TriggerRoute.Trigger Route (Int32, 2):Int32"(Set 0=^NPC.Event.PlayStage3Loop, Value 1=@n17.Value, Set 1=@n9.Value) {also {defaults {Value 0=0}}}
   n18 = VariableAccessor.Trigger(Variable=VariableMutator.Trigger(
     Variable=InitVariable.Trigger().Variable
     Value=&MSP_ScheduleTrigger(Schedule=@TriggerAnyTriggerAny2_1.Out, Triggers From Now=@TriggerRouteTriggerRouteInt322_0.Value, In Trigger=@n5.Value).Scheduled Trigger
    ).Variable)
   n19 = VariableAccessor.Trigger(Variable=@n18.Variable)
   n20 = VariableAccessor.Trigger(Variable=@n19.Variable)
   n21 = VariableAccessor.Trigger(Variable=@n20.Variable)
   TriggerAnyTriggerAny2_6 = "TriggerAny.Trigger Any (2)"(In 0=@n21.Value)
   n22 = VariableAccessor.Trigger(Variable=@n7.Variable)
   n23 = &MSP_ScheduleTrigger(Schedule=^NPC.Event.PlayTransitionToEnding, In Trigger=@n22.Value)
   TriggerAnyTriggerAny2_7 = "TriggerAny.Trigger Any (2)"(In 0=^NPC.Event.PlayStage1Loop)
   n24 = VariableAccessor.Trigger(Variable=VariableMutator.Trigger(
     Variable=InitVariable.Trigger().Variable
     Value=&MSP_ScheduleTrigger(Schedule=@TriggerAnyTriggerAny2_7.Out, In Trigger=@n6.Value).Scheduled Trigger
    ).Variable)
   n25 = VariableAccessor.Trigger(Variable=@n22.Variable)
   n26 = VariableAccessor.Trigger(Variable=@n25.Variable)
   n27 = VariableAccessor.Trigger(Variable=@n26.Variable)
   n28 = VariableAccessor.Trigger(Variable=@n27.Variable)
   n29 = VariableAccessor.Trigger(Variable=VariableMutator.Trigger(
     Variable=InitVariable.Trigger().Variable
     Value=&MSP_ScheduleTrigger(Schedule=^NPC.Event.PlayEnding03, In Trigger=@n28.Value).Scheduled Trigger
    ).Variable)
   n30 = VariableAccessor.Trigger(Variable=@n29.Variable)
   n31 = VariableAccessor.Trigger(Variable=@n30.Variable)
   n32 = VariableAccessor.Trigger(Variable=@n31.Variable)
   n33 = VariableAccessor.Trigger(Variable=@n32.Variable)
   n34 = VariableAccessor.Trigger(Variable=VariableMutator.Trigger(
     Variable=InitVariable.Trigger().Variable
     Value=&MSP_ScheduleTrigger(Triggers From Now=^EndingDelay, Schedule=^NPC.Event.StopAll, In Trigger=@n4.Value).Scheduled Trigger
    ).Variable)
   n35 = VariableAccessor.Trigger(Variable=@n34.Variable)
   n36 = &MSP_ScheduleTrigger(Schedule=^NPC.Event.PlayTransitionToStage2, In Trigger=@n2.Value)
   n37 = VariableAccessor.Trigger(Variable=VariableMutator.Trigger(Variable=InitVariable.Trigger().Variable, Value=@n36.Scheduled Trigger).Variable)
   TriggerAnyTriggerAny2_4 = "TriggerAny.Trigger Any (2)"(In 0=^NPC.Event.PlayStage2Loop, In 1=@n37.Value)
   TriggerRouteTriggerRouteInt322_1 = "TriggerRoute.Trigger Route (Int32, 2):Int32"(Set 0=^NPC.Event.PlayStage2Loop, Value 1=@n16.Value, Set 1=@n37.Value) {also {defaults {Value 0=0}}}
   n38 = VariableAccessor.Trigger(Variable=VariableMutator.Trigger(
     Variable=InitVariable.Trigger().Variable
     Value=&MSP_ScheduleTrigger(Schedule=@TriggerAnyTriggerAny2_4.Out, Triggers From Now=@TriggerRouteTriggerRouteInt322_1.Value, In Trigger=@n3.Value).Scheduled Trigger
    ).Variable)
   n39 = VariableAccessor.Trigger(Variable=VariableMutator.Trigger(
     Variable=InitVariable.Trigger().Variable
     Value=&MSP_ScheduleTrigger(Schedule=^NPC.Event.PlayEnding01, In Trigger=@n27.Value).Scheduled Trigger
    ).Variable)
   n40 = VariableAccessor.Trigger(Variable=@n39.Variable)
   n41 = VariableAccessor.Trigger(Variable=VariableMutator.Trigger(
     Variable=InitVariable.Trigger().Variable
     Value=&MSP_ScheduleTrigger(Schedule=^NPC.Event.PlayEnding02, In Trigger=@n26.Value).Scheduled Trigger
    ).Variable)
   n42 = VariableAccessor.Trigger(Variable=@n41.Variable)
   n43 = VariableAccessor.Trigger(Variable=VariableMutator.Trigger(
     Value=&MSP_ScheduleTrigger(Schedule=^NPC.Event.PlayTransitionToStage1, In Trigger=@n25.Value).Scheduled Trigger
     Variable=@n13.Variable
    ).Variable)
   n44 = VariableAccessor.Trigger(Variable=@n43.Variable)
   n45 = VariableAccessor.Trigger(Variable=@n44.Variable)
   n46 = VariableAccessor.Trigger(Variable=@n45.Variable)
   TriggerAnyTriggerAny8_1 = "TriggerAny.Trigger Any (8)"(
    In 0=@n24.Value
    In 5=VariableAccessor.Trigger(Variable=@n33.Variable).Value
    In 7=@n35.Value
    In 1=@n38.Value
    In 3=@n40.Value
    In 4=@n42.Value
    In 2=@n18.Value
    In 6=@n46.Value
   )
   n47 = VariableAccessor.Trigger(Variable=VariableMutator.Trigger(Value=@TriggerAnyTriggerAny8_1.Out, Variable=@n15.Variable).Variable)
   n48 = VariableAccessor.Float(Variable=VariableMutator.Float(
     Variable=InitVariable.Float().Variable
     Value=Value.Value:Float(Set=@n23.Scheduled Trigger, Reset=@n47.Value) {also {defaults {Init Value=1}}}.Output Value
    ).Variable)
   n49 = VariableAccessor.Int32(Variable=@n17.Variable)
   n50 = VariableAccessor.Int32(Variable=@n49.Variable)
   n51 = VariableAccessor.Trigger(Variable=VariableMutator.Trigger(Value=@n23.Scheduled Trigger, Variable=@n12.Variable).Variable)
   n52 = UE.Add:Float(PrimaryOperand=&MSP_Fade(Fade Time=/FortNPCMusic/Audio/Patches/TransitionLengthToSeconds(BPM=^BPM, TransitionLength=@n50.Value).Output, In=@n51.Value) {also {defaults {Curve Exp=0.7}}}.Value Out) {also {defaults {AdditionalOperands=0.2}}}
   n53 = VariableAccessor.Trigger(Variable=@n46.Variable)
   n54 = Value.Value:Float(Reset=VariableAccessor.Trigger(Variable=@n51.Variable).Value, Set=@n53.Value) {also {defaults {Init Value=1}}}
   n55 = VariableAccessor.Float(Variable=VariableMutator.Float(Variable=InitVariable.Float().Variable, Value=UE.Multiply:Float(AdditionalOperands=@n52.Out, PrimaryOperand=@n54.Output Value).Out).Variable)
   TriggerAnyTriggerAny2_2 = "TriggerAny.Trigger Any (2)"(In 0=VariableAccessor.Trigger(Variable=@n53.Variable).Value)
   n56 = InitVariable.Trigger()
   n57 = VariableAccessor.Trigger(Variable=@n42.Variable)
   n58 = VariableAccessor.Trigger(Variable=@n57.Variable)
   n59 = VariableAccessor.Trigger(Variable=@n58.Variable)
   n60 = VariableAccessor.Trigger(Variable=@n40.Variable)
   n61 = VariableAccessor.Trigger(Variable=@n60.Variable)
   n62 = VariableAccessor.Trigger(Variable=@n61.Variable)
   n63 = VariableAccessor.Trigger(Variable=@n38.Variable)
   n64 = VariableAccessor.Trigger(Variable=@n63.Variable)
   n65 = VariableAccessor.Trigger(Variable=@n64.Variable)
   TriggerAnyTriggerAny5_0 = "TriggerAny.Trigger Any (5)"(
    In 4=@n33.Value
    In 3=VariableAccessor.Trigger(Variable=@n59.Variable).Value
    In 2=VariableAccessor.Trigger(Variable=@n62.Variable).Value
    In 1=VariableAccessor.Trigger(Variable=@n21.Variable).Value
    In 0=VariableAccessor.Trigger(Variable=@n65.Variable).Value
   )
   n66 = Value.Value:Int32(
    Set=@TriggerAnyTriggerAny2_2.Out
    Target Value=@n49.Value
    Reset=VariableAccessor.Trigger(Variable=VariableMutator.Trigger(Variable=@n56.Variable, Value=@TriggerAnyTriggerAny5_0.Out).Variable).Value
   ) {also {defaults {Init Value=0}}}
   n67 = VariableAccessor.Float(Variable=VariableMutator.Float(
     Variable=InitVariable.Float().Variable
     Value=&TransitionLengthToSeconds(BPM=^BPM, TransitionLength=@n66.Output Value).Output
    ).Variable)
   n68 = InitVariable.Trigger()
   n69 = VariableAccessor.Trigger(Variable=@n24.Variable)
   n70 = VariableAccessor.Trigger(Variable=@n69.Variable)
   n71 = VariableAccessor.Trigger(Variable=@n35.Variable)
   TriggerAnyTriggerAny7_1 = "TriggerAny.Trigger Any (7)"(
    In 3=@n60.Value
    In 0=VariableAccessor.Trigger(Variable=@n70.Variable).Value
    In 5=@n32.Value
    In 2=@n45.Value
    In 1=@n65.Value
    In 4=@n57.Value
    In 6=VariableAccessor.Trigger(Variable=@n71.Variable).Value
   )
   n72 = /FortNPCMusic/Audio/Patches/MSP_LayeredTrackPlayback2(
    Stem2.Stem=^Loop.Stage03Else
    Stem1.Stem=^Loop.Stage03Perc
    Play=@TriggerAnyTriggerAny2_6.Out
    Stem2.Volume=@n48.Value
    Stem1.Volume=VariableAccessor.Float(Variable=@n55.Variable).Value
    FadeTime.Out=VariableAccessor.Float(Variable=@n67.Variable).Value
    Stop=VariableAccessor.Trigger(Variable=VariableMutator.Trigger(Variable=@n68.Variable, Value=@TriggerAnyTriggerAny7_1.Out).Variable).Value
   )
   TriggerAnyTriggerAny2_9 = "TriggerAny.Trigger Any (2)"(In 0=@n64.Value)
   n73 = InitVariable.Trigger()
   n74 = VariableAccessor.Trigger(Variable=VariableMutator.Trigger(Value=@n8.Scheduled Trigger, Variable=@n10.Variable).Variable)
   TriggerAnyTriggerAny8_0 = "TriggerAny.Trigger Any (8)"(In 7=@n71.Value, In 5=@n31.Value, In 3=@n61.Value, In 2=@n19.Value, In 4=@n58.Value, In 6=@n44.Value, In 0=@n69.Value, In 1=@n74.Value)
   n75 = &MSP_LayeredTrackPlayback2(
    Stem2.Stem=^Loop.Stage02Else
    Stem1.Stem=^Loop.Stage02Perc
    Play=@TriggerAnyTriggerAny2_9.Out
    Stem2.Volume=VariableAccessor.Float(Variable=@n48.Variable).Value
    FadeTime.Out=@n67.Value
    Stem1.Volume=@n55.Value
    Stop=VariableAccessor.Trigger(Variable=VariableMutator.Trigger(Variable=@n73.Variable, Value=@TriggerAnyTriggerAny8_0.Out).Variable).Value
   )
   n76 = VariableAccessor.Trigger(Variable=VariableMutator.Trigger(Value=@n36.Scheduled Trigger, Variable=@n11.Variable).Variable)
   WavePlayer_0 = UE.Wave Player:Stereo(Wave Asset=^Stinger.TransitionToStage02, Play=VariableAccessor.Trigger(Variable=@n76.Variable).Value)
   TriggerAnyTriggerAny2_3 = "TriggerAny.Trigger Any (2)"(In 0=@n39.Value)
   WavePlayer_1 = UE.Wave Player:Stereo(Wave Asset=^Stinger.Ending01, Play=@TriggerAnyTriggerAny2_3.Out)
   n77 = VariableDeferredAccessor.Trigger(Variable=InitVariable.Trigger().Variable)
   TriggerAnyTriggerAny3_0 = "TriggerAny.Trigger Any (3)"(In 0=@n70.Value, In 2=@n43.Value, In 1=@n77.Value)
   n78 = InitVariable.Trigger()
   TriggerAnyTriggerAny7_0 = "TriggerAny.Trigger Any (7)"(In 0=@n76.Value, In 5=@n30.Value, In 4=@n59.Value, In 1=@n63.Value, In 3=@n62.Value, In 2=@n20.Value, In 6=@n34.Value)
   WavePlayer_2 = UE.Wave Player:Stereo(
    Wave Asset=^Loop.Stage01Loop
    Play=@TriggerAnyTriggerAny3_0.Out
    Stop=VariableAccessor.Trigger(Variable=VariableMutator.Trigger(Variable=@n78.Variable, Value=@TriggerAnyTriggerAny7_0.Out).Variable).Value
   )
   TriggerAnyTriggerAny2_8 = "TriggerAny.Trigger Any (2)"(In 0=@n29.Value)
   WavePlayer_3 = UE.Wave Player:Stereo(Wave Asset=^Stinger.Ending03, Play=@TriggerAnyTriggerAny2_8.Out)
   TriggerAnyTriggerAny2_5 = "TriggerAny.Trigger Any (2)"(In 0=@n41.Value)
   WavePlayer_4 = UE.Wave Player:Stereo(Wave Asset=^Stinger.Ending02, Play=@TriggerAnyTriggerAny2_5.Out)
   TriggerAnyTriggerAny2_0 = "TriggerAny.Trigger Any (2)"(In 0=VariableAccessor.Trigger(Variable=@n74.Variable).Value)
   WavePlayer_5 = UE.Wave Player:Stereo(Wave Asset=^Stinger.TransitionToStage03, Play=@TriggerAnyTriggerAny2_0.Out)
   AudioMixerAudioMixerStereo8_0 = "AudioMixer.Audio Mixer (Stereo, 8)"(
    In 4 L=@n72.Out L
    In 4 R=@n72.Out R
    In 2 L=@n75.Out L
    In 2 R=@n75.Out R
    In 1 L=@WavePlayer_0.Out Left
    In 1 R=@WavePlayer_0.Out Right
    In 5 L=@WavePlayer_1.Out Left
    In 5 R=@WavePlayer_1.Out Right
    In 0 L=@WavePlayer_2.Out Left
    In 0 R=@WavePlayer_2.Out Right
    In 7 L=@WavePlayer_3.Out Left
    In 7 R=@WavePlayer_3.Out Right
    In 6 L=@WavePlayer_4.Out Left
    In 6 R=@WavePlayer_4.Out Right
    In 3 L=@WavePlayer_5.Out Left
    In 3 R=@WavePlayer_5.Out Right
   ) {also {defaults {Gain 0=3, Gain 7=1.2}}}
   Multiply_0 = UE.Multiply:Audio by Float(AdditionalOperands=^MasterVolume, PrimaryOperand=@AudioMixerAudioMixerStereo8_0.Out R)
   Multiply_1 = UE.Multiply:Audio by Float(AdditionalOperands=^MasterVolume, PrimaryOperand=@AudioMixerAudioMixerStereo8_0.Out L)
   VariableMutator.Trigger(Variable=@n77.Variable, Value=@WavePlayer_2.On Nearly Finished)
   VariableMutator.Trigger(Variable=InitVariable.Trigger().Variable, Value=@TriggerCounter_0.On Trigger)
  ]
  links [
   {dst="@UE.OutputFormat.Stereo.Audio:1", src=Multiply_0.Out}
   {dst="@UE.OutputFormat.Stereo.Audio:0", src=Multiply_1.Out}
   {dst="@CurrentPhase", src=TriggerRouteTriggerRouteInt325_0.Value}
  ]
 }]
properties {
 assetImportData=&self:AssetImportData
 bLooping=true
 duration=10000
 numChannels=2
 outputFormat=Stereo
 sampleRate=48000
 soundClassObject=&Fort_Music_NPC
 virtualizationMode=PlayWhenSilent
}
