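# MetaSound preset of MSS_GiftedTraveler_Music that supplies staged, spatially attenuated stereo music for the Gifted Traveler NPC.
# It fills the parent's inputs with three stage loops (later stages split into percussion and other stems), transition and ending stingers, plus tempo and bar timing.
# NPC event triggers select loops, transitions, endings or stop, and a CurrentPhase output reports state.
# The looping source keeps playing when silent; playback logic lives in the parent.
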MetaSoundSource /GiftedTraveler/Assets/Audio/MSS_GiftedTraveler_Music_Preset
preset_parent=&/MSS_GiftedTraveler_Music
inputs [
 AdditionalOperands {type=Float}
 Beats per Bar {type=Int32}
 BPM {type=Float}
 EndingDelay {type=Int32}
 Loop Length Bars {type=Int32}
 Loop.Stage01Loop {type=WaveAsset}
 Loop.Stage02Else {type=WaveAsset}
 Loop.Stage02Perc {type=WaveAsset}
 Loop.Stage03Else {type=WaveAsset}
 Loop.Stage03Perc {type=WaveAsset}
 MasterVolume {type=Float}
 NPC.Event.PlayEnding01 {type=Trigger}
 NPC.Event.PlayEnding02 {type=Trigger}
 NPC.Event.PlayEnding03 {type=Trigger}
 NPC.Event.PlayStage1Loop {type=Trigger}
 NPC.Event.PlayStage2Loop {type=Trigger}
 NPC.Event.PlayStage3Loop {type=Trigger}
 NPC.Event.PlayTransitionToEnding {type=Trigger}
 NPC.Event.PlayTransitionToStage1 {type=Trigger}
 NPC.Event.PlayTransitionToStage2 {type=Trigger}
 NPC.Event.PlayTransitionToStage3 {type=Trigger}
 NPC.Event.StopAll {type=Trigger}
 Stinger.Ending01 {type=WaveAsset}
 Stinger.Ending02 {type=WaveAsset}
 Stinger.Ending03 {type=WaveAsset}
 Stinger.TransitionToStage02 {type=WaveAsset}
 Stinger.TransitionToStage03 {type=WaveAsset}
 TransitionLengthsBars {type=Int32:Array}
 UE.Source.OnPlay {type=Trigger}
]
interfaces [UE.OutputFormat.Stereo:1.0, UE.Source:1.1]
outputs [
 CurrentPhase {type=Int32}
 "UE.OutputFormat.Stereo.Audio:0" {type=Audio}
 "UE.OutputFormat.Stereo.Audio:1" {type=Audio}
]
pages [Default {input_defaults {
   AdditionalOperands=2.06
   BPM=150
   Beats per Bar=4
   EndingDelay=1
   Loop Length Bars=16
   Loop.Stage01Loop=/GiftedTraveler/Assets/Audio/Music/MX_S34_GiftedTraveler_Stage_01_LP
   Loop.Stage02Else=/GiftedTraveler/Assets/Audio/Music/MX_S34_GiftedTraveler_Stage_02_LP_Else
   Loop.Stage02Perc=/GiftedTraveler/Assets/Audio/Music/MX_S34_GiftedTraveler_Stage_02_LP_Perc
   Loop.Stage03Else=/GiftedTraveler/Assets/Audio/Music/MX_S34_GiftedTraveler_Stage_03_LP_Else
   Loop.Stage03Perc=/GiftedTraveler/Assets/Audio/Music/MX_S34_GiftedTraveler_Stage_03_LP_Perc
   MasterVolume=0.7079
   NPC.Event.PlayEnding01=null
   NPC.Event.PlayEnding02=null
   NPC.Event.PlayEnding03=null
   NPC.Event.PlayStage1Loop=null
   NPC.Event.PlayStage2Loop=null
   NPC.Event.PlayStage3Loop=null
   NPC.Event.PlayTransitionToEnding=null
   NPC.Event.PlayTransitionToStage1=null
   NPC.Event.PlayTransitionToStage2=null
   NPC.Event.PlayTransitionToStage3=null
   NPC.Event.StopAll=null
   Stinger.Ending01=/GiftedTraveler/Assets/Audio/Music/MX_S34_GiftedTraveler_End_01_STNG
   Stinger.Ending02=/GiftedTraveler/Assets/Audio/Music/MX_S34_GiftedTraveler_End_02_STNG
   Stinger.Ending03=/GiftedTraveler/Assets/Audio/Music/MX_S34_GiftedTraveler_End_PortalEnter_STNG
   Stinger.TransitionToStage02=/GiftedTraveler/Assets/Audio/Music/MX_S34_GiftedTraveler_Stage_01_to_02_STNG
   Stinger.TransitionToStage03=/GiftedTraveler/Assets/Audio/Music/MX_S34_GiftedTraveler_Stage_02_to_03_STNG
   UE.Source.OnPlay=null
   TransitionLengthsBars [3,2]
  }}]
properties {
 assetImportData=&self:AssetImportData
 attenuationSettings=/GiftedTraveler/Assets/Audio/Music/ATT_MX_GiftedTraveler
 bLooping=true
 duration=10000
 maxDistance=8500
 numChannels=2
 outputFormat=Stereo
 sampleRate=48000
 soundClassObject=&Fort_Music_NPC
 virtualizationMode=PlayWhenSilent
}
