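# Sound cue for stage five of the GiftedTraveler Battle Bus beam audio.
# A single wave player feeds the output directly, playing a wave named "Bean_Stage_04", which differs from the cue's "Beam_Stage_05" name.
# It uses the shared Battle Bus attenuation asset with a very large max distance and routes to the Fort_Events sound class.
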SoundCue /GiftedTraveler/Assets/Audio/SC_GiftedTraveler_BattleBus_Beam_Stage_05
root=WavePlayer_2
nodes [{id=WavePlayer_2, properties {soundWaveAssetPtr=&/GiftedTraveler_BattleBus_Bean_Stage_04}}]
properties {
 attenuationSettings=&/Att_GiftedTraveler_BattleBus
 maxDistance=412000
 soundClassObject=&Fort_Events
 volumeMultiplier=1
 attenuationOverrides {bAttenuate=false, bEnableReverbSend=false, bSpatialize=false}
}
