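# Sound cue for the first stage of the Battle Bus explosion in GiftedTraveler, classed under Fort_Events.
# A single Wave Player plays the stage 01 explosion wave, with no randomization, mixing or modulation.
# It references the shared GiftedTraveler Battle Bus attenuation asset and carries a very large max distance.
# Its stored attenuation overrides switch off attenuation, spatialization and reverb send, but the override flag itself is not set.
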SoundCue /GiftedTraveler/Assets/Audio/SC_GiftedTraveler_BattleBus_Explosion_Stage_01
root=WavePlayer_1
nodes [{id=WavePlayer_1, properties {soundWaveAssetPtr=&/GiftedTraveler_BattleBus_Explosion_Stage_01}}]
properties {
 attenuationSettings=&/Att_GiftedTraveler_BattleBus
 maxDistance=412000
 soundClassObject=&Fort_Events
 volumeMultiplier=1
 attenuationOverrides {bAttenuate=false, bEnableReverbSend=false, bSpatialize=false}
}
