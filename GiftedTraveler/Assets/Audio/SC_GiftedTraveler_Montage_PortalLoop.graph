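# SoundCue for the GiftedTraveler montage portal loop, with no sound nodes and a null root, so the cue as shown references no audio.
# It carries only playback settings: a dedicated attenuation asset, the Athena ambient props sound class and a reduced volume multiplier.
# A concurrency override caps it at three simultaneous instances.
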SoundCue /GiftedTraveler/Assets/Audio/SC_GiftedTraveler_Montage_PortalLoop
root=null
nodes []
properties {
 attenuationSettings=&/Att_GiftedTraveler_Montage_PortalLoop
 bOverrideConcurrency=true
 maxDistance=4750
 soundClassObject=&Fort_Ambient_Props_Athena
 volumeMultiplier=0.6
 concurrencyOverrides {maxCount=3}
}
