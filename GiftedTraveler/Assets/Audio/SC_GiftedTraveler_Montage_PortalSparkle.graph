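# Sound cue for the portal sparkle in the Gifted Traveler montage, a spatialized ambient one-shot mix.
# A mixer layers random picks from whisper, electric one-shot and Sparks A wave pools, plus a Sparks C pick that enters after a fixed one-second delay.
# It references the PortalLoop_Hands attenuation, the Athena ambient props sound class and a shared loot ambient concurrency set.
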SoundCue /GiftedTraveler/Assets/Audio/SC_GiftedTraveler_Montage_PortalSparkle
root=Mixer_0
nodes [
 {id=Delay_0, children [Random_4], properties {delayMax=1, delayMin=1}}
 {id=Mixer_0, children [Random_0,Random_3,Random_1,Delay_0], properties {inputVolume [1.5,1.5,0.75,1.25]}}
 {id=Random_0, children [WavePlayer_15, WavePlayer_16, WavePlayer_17, WavePlayer_18, WavePlayer_19], properties {weights [1,1,1,1,1]}}
 {id=Random_1, children [WavePlayer_0,WavePlayer_1,WavePlayer_2,WavePlayer_3], properties {weights [1,1,1,1]}}
 {
  id=Random_3
  children [WavePlayer_20, WavePlayer_21, WavePlayer_22, WavePlayer_23, WavePlayer_24, WavePlayer_25, WavePlayer_26, WavePlayer_27]
  properties {weights [1,1,1,1,1,1,1,1]}
 }
 {id=Random_4, children [WavePlayer_11, WavePlayer_12, WavePlayer_13, WavePlayer_14, WavePlayer_28], properties {weights [1,1,1,1,1]}}
 {id=WavePlayer_0, properties {soundWaveAssetPtr=&/GiftedTraveler_Montage_Sparks_A_01}}
 {id=WavePlayer_1, properties {soundWaveAssetPtr=&/GiftedTraveler_Montage_Sparks_A_02}}
 {id=WavePlayer_11, properties {soundWaveAssetPtr=&/GiftedTraveler_Montage_Sparks_C_01}}
 {id=WavePlayer_12, properties {soundWaveAssetPtr=&/GiftedTraveler_Montage_Sparks_C_02}}
 {id=WavePlayer_13, properties {soundWaveAssetPtr=&/GiftedTraveler_Montage_Sparks_C_03}}
 {id=WavePlayer_14, properties {soundWaveAssetPtr=&/GiftedTraveler_Montage_Sparks_C_04}}
 {id=WavePlayer_15, properties {soundWaveAssetPtr=&/GiftedTraveler_Montage_Whispers_01}}
 {id=WavePlayer_16, properties {soundWaveAssetPtr=&/GiftedTraveler_Montage_Whispers_02}}
 {id=WavePlayer_17, properties {soundWaveAssetPtr=&/GiftedTraveler_Montage_Whispers_03}}
 {id=WavePlayer_18, properties {soundWaveAssetPtr=&/GiftedTraveler_Montage_Whispers_04}}
 {id=WavePlayer_19, properties {soundWaveAssetPtr=&/GiftedTraveler_Montage_Whispers_05}}
 {id=WavePlayer_2, properties {soundWaveAssetPtr=&/GiftedTraveler_Montage_Sparks_A_03}}
 {id=WavePlayer_20, properties {soundWaveAssetPtr=&/GiftedTraveler_Montage_ElecOneShots_01}}
 {id=WavePlayer_21, properties {soundWaveAssetPtr=&/GiftedTraveler_Montage_ElecOneShots_02}}
 {id=WavePlayer_22, properties {soundWaveAssetPtr=&/GiftedTraveler_Montage_ElecOneShots_03}}
 {id=WavePlayer_23, properties {soundWaveAssetPtr=&/GiftedTraveler_Montage_ElecOneShots_04}}
 {id=WavePlayer_24, properties {soundWaveAssetPtr=&/GiftedTraveler_Montage_ElecOneShots_05}}
 {id=WavePlayer_25, properties {soundWaveAssetPtr=&/GiftedTraveler_Montage_ElecOneShots_06}}
 {id=WavePlayer_26, properties {soundWaveAssetPtr=&/GiftedTraveler_Montage_ElecOneShots_07}}
 {id=WavePlayer_27, properties {soundWaveAssetPtr=&/GiftedTraveler_Montage_ElecOneShots_08}}
 {id=WavePlayer_28, properties {soundWaveAssetPtr=&/GiftedTraveler_Montage_Sparks_C_05}}
 {id=WavePlayer_3, properties {soundWaveAssetPtr=&/GiftedTraveler_Montage_Sparks_A_04}}
]
properties {
 attenuationSettings=&/Att_GiftedTraveler_Montage_PortalLoop_Hands
 maxDistance=2500
 priority=50
 soundClassObject=&Fort_Ambient_Props_Athena
 volumeMultiplier=1
 attenuationOverrides {
  bEnableFocusInterpolation=true
  bEnableListenerFocus=true
  distanceAlgorithm=Logarithmic
  falloffDistance=1100
  focusAttackInterpSpeed=20
  focusAzimuth=20
  focusReleaseInterpSpeed=10
  focusVolumeAttenuation=3.5
  lPFFrequencyAtMax=10000
  lPFRadiusMax=1000
  lPFRadiusMin=500
  nonFocusAzimuth=40
  nonFocusVolumeAttenuation=0.8
  nonSpatializedRadiusStart=200
  occlusionInterpolationTime=0.3
  occlusionLowPassFilterFrequency=3000
  occlusionVolumeAttenuation=0.75
  spatializationAlgorithm=SPATIALIZATION_HRTF
  attenuationShapeExtents {x=0}
  customAttenuationCurve {editorCurveData {keys [<at 0 1 RCIM_Cubic>, {arriveTangent=-1, interpMode=RCIM_Cubic, leaveTangent=-1, time=0.5, value=0.1}, <at 1 0 RCIM_Cubic>]}}
 }
 concurrencyOverrides {resolutionRule=StopQuietest}
 concurrencySet [/Game/Sounds/Concurrency_Rules/Loot_Ambient_Conc]
}
