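# Sound cue for a Gifted Traveler montage portal voice-over line, as the asset names indicate.
# A single wave player at the root plays the GiftedTraveler_Montage_VO_01 sound wave, with no randomization or mixing.
# The cue is spatialized through a close point-of-interest attenuation asset, routed to the Fort_Events sound class, and played with boosted volume.
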SoundCue /GiftedTraveler/Assets/Audio/SC_GiftedTraveler_Montage_PortalVO_01
root=WavePlayer_0
nodes [{id=WavePlayer_0, properties {soundWaveAssetPtr=&/GiftedTraveler_Montage_VO_01}}]
properties {
 attenuationSettings=&/ATT_GiftedTraveler_POI_Close
 maxDistance=14000
 soundClassObject=&Fort_Events
 volumeMultiplier=2
}
