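# Sound cue for the Gifted Traveler montage portal voice-over, second line.
# A single wave player feeds the output directly from the VO_02 montage sound wave, with no randomization or modulation.
# It uses the close point-of-interest attenuation and the Fort_Events sound class, with a boosted volume.
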SoundCue /GiftedTraveler/Assets/Audio/SC_GiftedTraveler_Montage_PortalVO_02
root=WavePlayer_1
nodes [{id=WavePlayer_1, properties {soundWaveAssetPtr=&/GiftedTraveler_Montage_VO_02}}]
properties {
 attenuationSettings=&/ATT_GiftedTraveler_POI_Close
 maxDistance=14000
 soundClassObject=&Fort_Events
 volumeMultiplier=1.5
}
