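; Gifted Traveler beam FX actor that plays Battle Bus burst and beam Niagara effects with sound during the Aircraft game phase.
; BeginPlay and a duplicate GT_BusBeam event chain delayed burst, beam and indicator spawns, passing the phase index as stage parameters.
; A VividRazor Baron router event fades out the indicator, while fixed timers destroy burst and beam components.
; OnGroundFXLoop runs a similar ground burst and beam sequence, commented as disabled by design.
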(event Custom|GT_BusBeam
  (bind _phaseindexforfx (Variables|Default|GetPhaseIndexForFX))
  (bind _returnvalue (Math|Integer|Clamp(Integer) _phaseindexforfx 1 4))
  (bind _returnvalue_2 (Math|Conversions|ToFloat(Integer) _returnvalue))
  (bind _initialbeamdelay (Variables|Default|GetInitialBeamDelay))
  (bind _returnvalue_3 (ScalableFloat|GetValueatLevel _initialbeamdelay _returnvalue_2))
  (bind _beamindicatordelay (Variables|Default|GetBeamIndicatorDelay))
  (bind _returnvalue_4 (ScalableFloat|GetValueatLevel _beamindicatordelay))
  (bind _initialburstdelay (Variables|Default|GetInitialBurstDelay))
  (bind _returnvalue_5 (ScalableFloat|GetValueatLevel _initialburstdelay))
  (bind _phaseindexforfx_2 (Variables|Default|GetPhaseIndexForFX))
  (bind _phaseindexforfx_3 (Variables|Default|GetPhaseIndexForFX))
  (bind _beamindicatorinstance (Variables|Default|GetBeamIndicatorInstance))
  (if (> _returnvalue_5 0)
    (Utilities|FlowControl|Delay _returnvalue_5)
    (else
      _))
  (bind _returnvalue_6 (Niagara|SpawnSystemAttached (Variables|Default|GetBattleBusBurstFX) (Variables|Default|GetDefaultSceneRoot)))
  (Audio|SpawnSoundAttached (Variables|Default|GetBusPhaseBurstSFX) _returnvalue_6)
  (Utilities|FlowControl|Sequence
    (:then_0
      (if (> _returnvalue_3 0)
        (Utilities|FlowControl|Delay _returnvalue_3)
        (else
          _))
      (bind _returnvalue_7 (Niagara|SpawnSystemAttached (Variables|Default|GetBattleBusBeamFX) (Variables|Default|GetDefaultSceneRoot)))
      (Audio|SpawnSoundAttached (Variables|Default|GetBusPhaseBeamSFX) _returnvalue_7)
      (Utilities|FlowControl|Sequence
        (:then_1
          ; Make sure FX be destory
          (Utilities|FlowControl|Delay 21.0)
          (Components|Activation|Deactivate _returnvalue_7)
          (Components|DestroyComponent _returnvalue_7))
        (:then_2
          ; FXPhaseSetup
          (Effects|Components|ParticleSystem|SetBoolParameter _returnvalue_7 "Stage4" (== _phaseindexforfx_2 3))
          (Effects|Components|ParticleSystem|SetIntParameter _returnvalue_7 "StageIndex" _phaseindexforfx_2)
          (Effects|Components|ParticleSystem|SetIntParameter _returnvalue_7 "Stage" _phaseindexforfx_2))
        (:then_3
          (if (> _returnvalue_4 0)
            (Utilities|FlowControl|Delay _returnvalue_4)
            (else
              _))
          (bind _returnvalue_8 (Niagara|SpawnSystemAttached (Variables|Default|GetBeamIndicatorFX) (Variables|Default|GetDefaultSceneRoot)))
          (Variables|Default|SetBeamIndicatorInstance _returnvalue_8)
          (bind (_asynctaskproxy _payload) (GameplayEventRouter|StartListeningToStatefulEvent (GameplayEventRouter|GetGlobalEventRouter) "/VividRazor_Baron/Gameplay/EventRouter/VividRazor_Baron_Event.VividRazor_Baron_Event")
            (:OnEventReceived
              (if (Utilities|Enum|Equal(Enum) (Utilities|Struct|BreakVividRazorBaronEvent _payload) "NewEnumerator5")
                (Niagara|SetNiagaraVariable(Bool) _beamindicatorinstance "FadeOut" true)
                (Utilities|FlowControl|Delay 1.0)
                (Components|Activation|Deactivate _beamindicatorinstance)
                (Development|PrintString (Utilities|String|Append "DEACTIVATE : " (Utilities|GetDisplayName _beamindicatorinstance)) true true "(R=1.000000,G=0.781034,B=0.015852,A=1.000000)" 8.0)))))))
    (:then_1
      ; Make sure FX be destory
      (Utilities|FlowControl|Delay 7.0)
      (Components|Activation|Deactivate _returnvalue_6)
      (Components|DestroyComponent _returnvalue_6))
    (:then_2
      ; FXPhaseSetup
      (Effects|Components|ParticleSystem|SetBoolParameter _returnvalue_6 "Stage4" (== _phaseindexforfx_3 3))
      (Effects|Components|ParticleSystem|SetIntParameter _returnvalue_6 "StageIndex" _phaseindexforfx_3)
      (Effects|Components|ParticleSystem|SetIntParameter _returnvalue_6 "Stage" _phaseindexforfx_3))))

(event EventBeginPlay
  (bind _phaseindexforfx (Variables|Default|GetPhaseIndexForFX))
  (bind _returnvalue (Math|Integer|Clamp(Integer) _phaseindexforfx 1 4))
  (bind _returnvalue_2 (Math|Conversions|ToFloat(Integer) _returnvalue))
  (bind _initialbeamdelay (Variables|Default|GetInitialBeamDelay))
  (bind _returnvalue_3 (ScalableFloat|GetValueatLevel _initialbeamdelay _returnvalue_2))
  (bind _beamindicatordelay (Variables|Default|GetBeamIndicatorDelay))
  (bind _returnvalue_4 (ScalableFloat|GetValueatLevel _beamindicatordelay))
  (bind _initialburstdelay (Variables|Default|GetInitialBurstDelay))
  (bind _returnvalue_5 (ScalableFloat|GetValueatLevel _initialburstdelay))
  (bind _phaseindexforfx_2 (Variables|Default|GetPhaseIndexForFX))
  (bind _phaseindexforfx_3 (Variables|Default|GetPhaseIndexForFX))
  (bind _beamindicatorinstance (Variables|Default|GetBeamIndicatorInstance))
  (bind _returnvalue_6 (GameplayEventRouter|GetEventRouter "/Script/FortniteGame.FortGameStateAthena" self))
  (bind (_asynctaskproxy _payload) (GameplayEventRouter|StartListeningToStatefulEvent _returnvalue_6 "/Script/FortniteGame.FortGamePhaseUpdatedEvent")
    (:OnEventReceived
      _)
    (:OnSavedEventStateExists
      _))
  (switch Utilities|FlowControl|Switch|SwitchonEAthenaGamePhase (Utilities|Struct|BreakFortGamePhaseUpdatedEvent _payload)
    (:Aircraft
      (if (> _returnvalue_5 0)
        (Utilities|FlowControl|Delay _returnvalue_5)
        (else
          _))
      (bind _returnvalue_7 (Niagara|SpawnSystemAttached (Variables|Default|GetBattleBusBurstFX) (Variables|Default|GetDefaultSceneRoot)))
      (Audio|SpawnSoundAttached (Variables|Default|GetBusPhaseBurstSFX) _returnvalue_7)
      (Utilities|FlowControl|Sequence
        (:then_0
          (if (> _returnvalue_3 0)
            (Utilities|FlowControl|Delay _returnvalue_3)
            (else
              _))
          (bind _returnvalue_8 (Niagara|SpawnSystemAttached (Variables|Default|GetBattleBusBeamFX) (Variables|Default|GetDefaultSceneRoot)))
          (Audio|SpawnSoundAttached (Variables|Default|GetBusPhaseBeamSFX) _returnvalue_8)
          (Utilities|FlowControl|Sequence
            (:then_1
              ; Make sure FX be destory
              (Utilities|FlowControl|Delay 21.0)
              (Components|Activation|Deactivate _returnvalue_8)
              (Components|DestroyComponent _returnvalue_8))
            (:then_2
              ; FXPhaseSetup
              (Effects|Components|ParticleSystem|SetBoolParameter _returnvalue_8 "Stage4" (== _phaseindexforfx_2 3))
              (Effects|Components|ParticleSystem|SetIntParameter _returnvalue_8 "StageIndex" _phaseindexforfx_2)
              (Effects|Components|ParticleSystem|SetIntParameter _returnvalue_8 "Stage" _phaseindexforfx_2))
            (:then_3
              (if (> _returnvalue_4 0)
                (Utilities|FlowControl|Delay _returnvalue_4)
                (else
                  _))
              (bind _returnvalue_9 (Niagara|SpawnSystemAttached (Variables|Default|GetBeamIndicatorFX) (Variables|Default|GetDefaultSceneRoot)))
              (Variables|Default|SetBeamIndicatorInstance _returnvalue_9)
              (bind (_asynctaskproxy_2 _payload_2) (GameplayEventRouter|StartListeningToStatefulEvent (GameplayEventRouter|GetGlobalEventRouter) "/VividRazor_Baron/Gameplay/EventRouter/VividRazor_Baron_Event.VividRazor_Baron_Event")
                (:OnEventReceived
                  (if (Utilities|Enum|Equal(Enum) (Utilities|Struct|BreakVividRazorBaronEvent _payload_2) "NewEnumerator5")
                    (Niagara|SetNiagaraVariable(Bool) _beamindicatorinstance "FadeOut" true)
                    (Utilities|FlowControl|Delay 1.0)
                    (Components|Activation|Deactivate _beamindicatorinstance)
                    (Development|PrintString (Utilities|String|Append "DEACTIVATE : " (Utilities|GetDisplayName _beamindicatorinstance)) true true "(R=1.000000,G=0.781034,B=0.015852,A=1.000000)" 8.0)))))))
        (:then_1
          ; Make sure FX be destory
          (Utilities|FlowControl|Delay 7.0)
          (Components|Activation|Deactivate _returnvalue_7)
          (Components|DestroyComponent _returnvalue_7))
        (:then_2
          ; FXPhaseSetup
          (Effects|Components|ParticleSystem|SetBoolParameter _returnvalue_7 "Stage4" (== _phaseindexforfx_3 3))
          (Effects|Components|ParticleSystem|SetIntParameter _returnvalue_7 "StageIndex" _phaseindexforfx_3)
          (Effects|Components|ParticleSystem|SetIntParameter _returnvalue_7 "Stage" _phaseindexforfx_3))))))

(event Custom|OnGroundFXLoop
  (bind _groundbeamdelay (Variables|Default|GetGroundBeamDelay))
  (bind _returnvalue (ScalableFloat|GetValueatLevel _groundbeamdelay))
  (bind _phaseindexforfx (Variables|Default|GetPhaseIndexForFX))
  (bind _phaseindexforfx_2 (Variables|Default|GetPhaseIndexForFX))
  ;[1 0 0.3137 0.5] Disable by design
  (bind _returnvalue_2 (Niagara|SpawnSystemAttached (Variables|Default|GetGroundBurstFX) (Variables|Default|GetDefaultSceneRoot)))
  (Utilities|FlowControl|Sequence
    (:then_0
      (if (> _returnvalue 0)
        (Utilities|FlowControl|Delay _returnvalue)
        (else
          _))
      (bind _returnvalue_3 (Niagara|SpawnSystemAttached (Variables|Default|GetGroundBeamFX) (Variables|Default|GetDefaultSceneRoot)))
      (Utilities|FlowControl|Sequence
        (:then_0
          ;; Make sure FX be destory
          (Utilities|FlowControl|Delay 19.0)
          (Components|Activation|Deactivate _returnvalue_3)
          (Components|DestroyComponent _returnvalue_3))
        (:then_1
          ;; FXPhaseSetup
          (Effects|Components|ParticleSystem|SetBoolParameter _returnvalue_3 "Stage4" (== _phaseindexforfx_2 3))
          (Effects|Components|ParticleSystem|SetIntParameter _returnvalue_3 "StageIndex" _phaseindexforfx_2)
          (Effects|Components|ParticleSystem|SetIntParameter _returnvalue_3 "Stage" _phaseindexforfx_2))))
    (:then_1
      ;; Make sure FX be destory
      (Utilities|FlowControl|Delay 7.0)
      (Components|Activation|Deactivate _returnvalue_2)
      (Components|DestroyComponent _returnvalue_2))
    (:then_2
      ;; FXPhaseSetup
      (Effects|Components|ParticleSystem|SetBoolParameter _returnvalue_2 "Stage4" (== _phaseindexforfx 3))
      (Effects|Components|ParticleSystem|SetIntParameter _returnvalue_2 "StageIndex" _phaseindexforfx)
      (Effects|Components|ParticleSystem|SetIntParameter _returnvalue_2 "Stage" _phaseindexforfx))))

(fn ConstructionScript)
