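# Niagara system for the GiftedTraveler gate: a magenta portal effect with a glowing center, mesh ribbons, lightning, smoke and sparks.
# System ActivationAlpha intensifies the effect by driving spawn rates, lifetimes, glow scale, color and forces, while FadeOut dims it and throttles most spawning.
# The flare emitter combines sprites, meshes and a light.
# Smoke and dense sparks simulate on GPU, with vortex and curl-noise swirl on the sparks; Center's portal-loop audio modules are disabled.
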NiagaraSystem /GiftedTraveler/Assets/FX/NE_GiftedTraveler_GateVFX
dependencies {
 usedDataInterfaces [
  &NiagaraDataInterfaceCurve
  &NiagaraDataInterfaceCurlNoise
  &NiagaraDataInterfaceColorCurve
  &NiagaraDataInterfaceAudioPlayer
 ]
 usedDynamicInputs [&MakeQuatFromAxisAngle, &OneMinusFloat]
 usedModules [
  &self:SystemSpawnScript.SystemScriptSource.SystemScriptGraph.NiagaraNodeAssignment_0.SetVariables_#1
  &SystemState
  &Timeline
  &self:SystemSpawnScript.SystemScriptSource.SystemScriptGraph.NiagaraNodeAssignment_1.SetVariables_#2
  &self:SystemSpawnScript.SystemScriptSource.SystemScriptGraph.NiagaraNodeAssignment_2.SetVariables_#3
 ]
}
emitters {
 Center {
  emitterSpawnScript [
   &InitializeAudioPlayer(Audio Player=DataInterface() {propertyValues {
      Attenuation=null
      Concurrency=null
      MaxPlaysPerTick=10
      SoundToPlay=/Hermes/Props/ClayTile/ClayTile_DragonGate_B/Audio/SC_DemonArena_Portal_Loop
      bAllowLoopingOneShotSounds=false
      bLimitPlaysPerTick=true
      bOnlyActiveDuringGameplay=true
      bStopWhenComponentIsDestroyed=true
      ConfigurationUserParameter {parameter {name=None, typeDefHandle {registeredTypeIndex=1}, varData []}}
      ParameterNames []
     }}) {enabled=false}
   /Niagara/Modules/Audio/PlayPersistentAudio() {enabled=false}
  ]
  emitterUpdateScript [
   &EmitterState(Life Cycle Mode=::NewEnumerator1, Loop Duration=1, UseLoopDelay=false)
   &SpawnRate(SpawnRate=&Lerp_Float(A=0.75, B=6, Alpha=$System.ActivationAlpha), Use Spawn Probability=false)
   /Niagara/Modules/Audio/UpdatePersistentAudio() {enabled=false}
  ]
  particleSpawnScript [
   /Niagara/Modules/Spawn/Initialization/InitializeParticle(
    Color=<rgba 0.0727 0.7686 1 1>
    Lifetime=8
    Module.Write Color=true
    Module.Write Lifetime=true
    Module.Write Mass=true
    Module.Write Material Random=true
    Module.Write Position=false
    Module.Write Scale=false
    Module.Write SpriteRotation=true
    Module.Write SpriteSize=true
    Position=&SimulationPosition()
    Sprite Rotation=&UniformRangedFloat(Minimum=0, Maximum=360)
    Sprite Size=<550 550>
   ) {enabled=false}
   /Niagara/Modules/Spawn/Initialization/V2/InitializeParticle(
    Color=<rgba 2 0.5521 0.9382 1>
    Lifetime=&Lerp_Float(A=8, B=0.6, Alpha=$System.ActivationAlpha)
    Mass Mode=::NewEnumerator1
    Sprite Rotation Angle=&UniformRangedFloat(Minimum=0, Maximum=360)
    Sprite Rotation Mode=::NewEnumerator0
    Sprite Size=<650 650>
    Sprite Size Mode=::NewEnumerator4
    UsePositionOffset=false
    Write Material Random=true
   ) {module=InitializeParticle001}
   /SetVariables(
    Particles.Position=&Add_Vector(A=Linked() {linkedVariable {name=Particles.Position, type=NiagaraPosition}}, B=<0 0 0>)
    Particles.Random=&UniformRangedFloat(Minimum=-0.1, Maximum=0.9)
   ) {isSetParameters=true}
  ]
  particleUpdateScript [
   &ParticleState()
   &ScaleSpriteSize(Scale Sprite Size Mode=::NewEnumerator0, Uniform Scale Factor=4)
   &SolveForcesAndVelocity(Clamp Velocity=false, Limit Acceleration=false)
   &DynamicMaterialParameters(
    Index 0 Param 1=&FloatFromCurve(FloatCurve=DataInterface() {propertyValues {Curve {keys [
         {arriveTangent=1.1486, interpMode=RCIM_Cubic, leaveTangent=1.1486, tangentMode=RCTM_User, time=0, value=0}
         {arriveTangent=-9.5651, interpMode=RCIM_Cubic, leaveTangent=-9.5651, tangentMode=RCTM_User, time=0.6446, value=-2.4156}
         {arriveTangent=-1.1502, interpMode=RCIM_Cubic, leaveTangent=-1.1502, tangentMode=RCTM_User, time=1, value=-3.0414}
        ]}}})
    Index 0 Param 2=&FloatFromCurve(FloatCurve=DataInterface() {propertyValues {Curve {keys [
         {arriveTangent=-10.815, interpMode=RCIM_Cubic, leaveTangent=-10.815, tangentMode=RCTM_User, time=0.0013, value=0.9969}
         {arriveTangent=-0.881, interpMode=RCIM_Cubic, leaveTangent=-0.881, time=0.3186, value=0.0305}
         {arriveTangent=2.4595, interpMode=RCIM_Cubic, leaveTangent=2.4595, time=0.758, value=0.3302}
         {arriveTangent=3.5097, interpMode=RCIM_Cubic, leaveTangent=3.5097, tangentMode=RCTM_User, time=1, value=1.7064}
        ]}}})
    Index 0 Param 4=&Add_Float(
     A=$Particles.Random
     B=&FloatFromCurve(
      FloatCurve=DataInterface() {propertyValues {Curve {keys [{interpMode=RCIM_Cubic, tangentMode=RCTM_User, time=0, value=0}, {interpMode=RCIM_Cubic, tangentMode=RCTM_User, time=1, value=1}]}}}
      Scale Curve=2
     )
    )
    Module.Param0WriteEnabled=true
    Module.Param1WriteEnabled=true
    Module.Param2WriteEnabled=true
    Module.Param3WriteEnabled=true
   )
   &ScaleColor(
    Scale Mode=::NewEnumerator1
    Scale RGBA=/Niagara/DynamicInputs/Multiply/Multiply_Vector4_ByFloat(
     Vector4:Vector4f=&Vector4FromCurve(Vector4Curve=DataInterface() {propertyValues {
        AlphaCurve {keys [
          <at 0.0037 0 RCIM_Cubic>
          {arriveTangent=2.9787, interpMode=RCIM_Cubic, leaveTangent=2.9787, time=0.1552, value=0.2595}
          {arriveTangent=1.0157, interpMode=RCIM_Cubic, leaveTangent=1.0157, time=0.2714, value=0.7974}
          {arriveTangent=-1.0205, interpMode=RCIM_Cubic, leaveTangent=-1.0205, time=0.7097, value=0.8228}
          {arriveTangent=-2.8344, interpMode=RCIM_Cubic, leaveTangent=-2.8344, time=0.8433, value=0.2138}
          <at 1 0 RCIM_Cubic>
         ]}
        BlueCurve {keys [<at 0.2703 1>,<at 0.8712 0>]}
        GreenCurve {keys [<at 0.2703 1>,<at 0.8712 0>]}
        RedCurve {keys [<at 0.2703 1>,<at 0.8712 0>]}
       }})
     Float:NiagaraFloat=$System.FadeOut
    )
    ScaleRGBA=true
   )
   &CameraOffset() {enabled=false}
   &self:MatchScale() {module=MatchScale}
  ]
  rendererClasses [&NiagaraSpriteRendererProperties]
  renderers [{rendererClass=&NiagaraSpriteRendererProperties}]
 }
 MeshRibbon {
  emitterSpawnScript [/SetVariables(Emitter.MeshRibbonData=DataInterface() {propertyValues {
      bExposeCurve=true
      AlphaCurve {keys []}
      BlueCurve {keys [<at 0 1>,<at 1 1>]}
      GreenCurve {keys [<at 0 -0.5 RCIM_Cubic>, {arriveTangent=0.0331, interpMode=RCIM_Cubic, leaveTangent=0.0331, tangentMode=RCTM_User, time=1, value=-1}]}
      RedCurve {keys [<at 0 1 RCIM_Cubic>,<at 1 2 RCIM_Cubic>]}
     }}) {isSetParameters=true}]
  emitterUpdateScript [
   &EmitterState()
   &SpawnBurst_Instantaneous(Spawn Count=12, Use Spawn Probability=false, UseLoopCountLimit=false) {enabled=false}
   &SpawnRate(
    SpawnRate=&Multiply_Float(A=&Lerp_Float(A=2, B=10, Alpha=$System.ActivationAlpha), B=$System.FadeOut)
    Use Spawn Probability=false
   )
   &SpawnRate(
    SpawnGroup=1
    SpawnRate=&Multiply_Float(A=&Lerp_Float(A=2, B=10, Alpha=$System.ActivationAlpha), B=$System.FadeOut)
    Use Spawn Probability=false
   ) {module=SpawnRate001}
  ]
  particleSpawnScript [
   /SetVariables(Particles.VisibilityTag=&SpawnGroup()) {isSetParameters=true}
   &InitializeParticle(
    Color=&MaskLinearColorBySpawnGroup(Masked Linear Color=<rgba 5 5 5 1>, Passthrough Linear Color=<rgba 1 1 1 11>)
    Lifetime=&MaskFloatBySpawnGroup(
     Masked Float=&Lerp_Float(
      A=&RandomRangeFloat(Minimum=1.8, Maximum=3.5)
      B=&RandomRangeFloat(Minimum=1, Maximum=1.5)
      Alpha=$System.ActivationAlpha
     )
     Passthrough Float=&Lerp_Float(
      A=&RandomRangeFloat(Minimum=5.5, Maximum=8.5)
      B=&RandomRangeFloat(Minimum=1.5, Maximum=2)
      Alpha=$System.ActivationAlpha
     )
    )
    Mesh Scale=&MaskVectorBySpawnGroup(
     Masked Vector=&Multiply_VectorByFloat(Vector=<1 1 2.5>, Float=&Lerp_Float(A=7, B=8, Alpha=$System.ActivationAlpha))
     Passthrough Vector=&Multiply_VectorByFloat(Vector=<1 -1 1>, Float=&RandomRangeFloat(Minimum=7.9909, Maximum=12.017))
    )
    Mesh Scale Mode=::NewEnumerator4
    UsePositionOffset=false
    Write Material Random=true
   )
   &InitialMeshOrientation(
    Mesh Orientation Mode=::NewEnumerator2
    Orientation Vector=<0 0 -1>
    Rotation=&RandomRangeVector(Minimum=<0 0 0>, Maximum=<0 0 350>)
    Use Rotation Vector=true
   )
  ]
  particleUpdateScript [
   &ParticleState()
   &ScaleMeshSize(Scale Factor=&VectorFromFloat(Value=&MaskFloatBySpawnGroup(
      Masked Float=&FloatFromCurve(FloatCurve=DataInterface() {propertyValues {Curve {keys [<at 0 1 RCIM_Cubic>,<at 1 0.9 RCIM_Cubic>]}}})
      Passthrough Float=&FloatFromCurve(FloatCurve=DataInterface() {propertyValues {Curve {keys [{arriveTangent=0.8366, interpMode=RCIM_Cubic, leaveTangent=0.8366, tangentMode=RCTM_User, time=0, value=0.5}, <at 1 0.75 RCIM_Cubic>]}}})
     )))
   &DynamicMaterialParameters(
    Index 0 Param 1=&FloatFromCurve(FloatCurve=DataInterface() {propertyValues {Curve {keys [
         {interpMode=RCIM_Cubic, tangentMode=RCTM_User, time=0, value=0.5}
         {arriveTangent=2.7589, interpMode=RCIM_Cubic, leaveTangent=2.7589, tangentMode=RCTM_User, time=1, value=2}
        ]}}})
    Index 0 Param 2=&FloatFromCurve(FloatCurve=DataInterface() {propertyValues {Curve {keys [{interpMode=RCIM_Cubic, tangentMode=RCTM_User, time=0, value=0}, {interpMode=RCIM_Cubic, tangentMode=RCTM_User, time=1, value=1}]}}})
    Index 0 Param 3=&RandomRangeFloat(Minimum=3.0512, Maximum=5.4)
    Index 0 Param 4=&Add_Float(A=&RandomRangeFloat(Minimum=16, Maximum=40), B=0)
    Module.Index0Param0Write=true
    Module.Index0Param1Write=true
    Module.Index0Param2Write=true
    Module.Index0Param3Write=true
   )
   &ScaleColor(
    Scale Mode=::NewEnumerator1
    Scale RGBA=/Niagara/DynamicInputs/Helpers/MaskVector4(
     Mask Parameter:NiagaraInt32=$Emitter.SpawnGroup
     Mask:NiagaraInt32=0
     Masked Vector4:Vector4f=&Vector4FromCurve(Vector4Curve=DataInterface() {propertyValues {
        AlphaCurve {keys [<at 0 0 RCIM_Cubic>,<at 0.2932 3 RCIM_Cubic>,<at 1 0 RCIM_Cubic>]}
        BlueCurve {keys [<at 0 0 RCIM_Cubic>,<at 0.2925 3 RCIM_Cubic>,<at 1 0 RCIM_Cubic>]}
        GreenCurve {keys [<at 0 0 RCIM_Cubic>,<at 0.2925 3 RCIM_Cubic>,<at 1 0 RCIM_Cubic>]}
        RedCurve {keys [<at 0 0 RCIM_Cubic>,<at 0.2925 3 RCIM_Cubic>,<at 1 0 RCIM_Cubic>]}
       }})
     Passthrough Vector4:Vector4f=&Vector4FromCurve(Vector4Curve=DataInterface() {propertyValues {
        AlphaCurve {keys [<at 0 0 RCIM_Cubic>,<at 0.6722 5 RCIM_Cubic>,<at 1 0 RCIM_Cubic>]}
        BlueCurve {keys [<at 0 0 RCIM_Cubic>,<at 0.6722 5 RCIM_Cubic>,<at 1 0 RCIM_Cubic>]}
        GreenCurve {keys [<at 0 0 RCIM_Cubic>,<at 0.6722 5 RCIM_Cubic>,<at 1 0 RCIM_Cubic>]}
        RedCurve {keys [<at 0 0 RCIM_Cubic>,<at 0.6722 5 RCIM_Cubic>,<at 1 0 RCIM_Cubic>]}
       }})
    )
    ScaleRGBA=true
   )
   &CameraOffset(Camera Offset Amount=50)
   /SetVariables(Particles.DarkRibbonDynMat:Vector4f=&MakeVector4(
     X=&FloatFromCurve(FloatCurve=DataInterface() {propertyValues {Curve {keys [{tangentMode=RCTM_User, time=0, value=-1}, {tangentMode=RCTM_User, time=1, value=0}]}}})
     Y=&FloatFromCurve(FloatCurve=DataInterface() {propertyValues {Curve {keys [{tangentMode=RCTM_User, time=0, value=0}, {tangentMode=RCTM_User, time=1, value=1}]}}})
     Z=&RandomRangeFloat(Minimum=2.5103, Maximum=7.4856)
     W=&Add_Float(A=&RandomRangeFloat(Minimum=25.1137, Maximum=44.0696), B=0)
    )) {isSetParameters=true}
  ]
  rendererClasses [&NiagaraMeshRendererProperties]
  renderers [
   {rendererClass=&NiagaraMeshRendererProperties, rendererIndex=0}
   {rendererClass=&NiagaraMeshRendererProperties, rendererIndex=1}
  ]
 }
 NE_GlowFlare {
  emitterUpdateScript [
   &EmitterState(
    Inactive Response=::NewEnumerator1
    Life Cycle Mode=::NewEnumerator1
    Loop Behavior=::NewEnumerator1
    Loop Duration Mode=::NewEnumerator1
    UseLoopDelay=false
   )
   &SpawnBurst_Instantaneous(Use Spawn Probability=false, UseLoopCountLimit=false)
  ]
  particleSpawnScript [
   &InitializeParticle(
    Color Mode=::NewEnumerator0
    Lifetime=2
    Mesh Scale Mode=::NewEnumerator3
    Sprite Size Mode=::NewEnumerator3
    Uniform Sprite Size=1500
    UsePositionOffset=true
    Write Material Random=true
   )
   &InitialMeshOrientation(Orientation Axis=<0 0 1>, Orientation Vector=<0 0 1>, Rotation=<0 0.25 0>, Use Orientation Vector=true, Use Rotation Vector=true)
   &CameraOffset(Camera Offset Amount=-150)
   /SetVariables(Particles.MeshCameraOffset=-1500) {isSetParameters=true}
   &SpriteFacingAndAlignment(Sprite Alignment=<0 0 1>, WriteAlignment=true, WriteFacing=false)
  ]
  particleUpdateScript [
   &ParticleState(Kill Particles When Lifetime Has Elapsed=false)
   &ScaleSpriteSize(
    Uniform Curve Index=$System.ActivationAlpha
    Uniform Curve Sprite Scale=DataInterface() {propertyValues {Curve {keys [
        {arriveTangent=2.1756, interpMode=RCIM_Cubic, leaveTangent=2.1756, tangentMode=RCTM_User, time=0, value=0}
        {interpMode=RCIM_Cubic, tangentMode=RCTM_User, time=1, value=1}
       ]}}}
   )
   &Color(Color=&ColorFromCurve(
     ColorCurve=DataInterface() {propertyValues {
       AlphaCurve {keys [<at 0 0>,<at 1 1 RCIM_Cubic>]}
       BlueCurve {keys [<at 0 0.2497 RCIM_Cubic>,<at 1 2.3687 RCIM_Cubic>]}
       GreenCurve {keys [<at 0 0 RCIM_Cubic>,<at 1 0.6997 RCIM_Cubic>]}
       RedCurve {keys [<at 0 1 RCIM_Cubic>,<at 1 5 RCIM_Cubic>]}
      }}
     CurveIndex=&Multiply_Float(A=$System.ActivationAlpha, B=$System.FadeOut)
    ))
   &FN_Sprite_Attributes(
    Color=&MultiplyLinearColor(Linear Color=$Particles.Color, Scale Factor=&MakeVector4_FromLinearColor(LinearColor=<rgba 1 1 1 0.5>))
    Create Color Attribute=true
    Create SpriteSize Attribute=true
    SpriteSize=&Multiply_Vector2D_ByFloat(Float=1, Vector2D=$Particles.SpriteSize)
    Write to Module Namespace=true
   )
   &FN_Sprite_Attributes(
    Color=&MultiplyLinearColor(
     Linear Color=$Particles.Color
     Scale Factor=&MakeVector4_FromLinearColor(LinearColor=<rgba 0.02 0.02 0.02 1>)
    )
    Create Color Attribute=true
    Create SpriteSize Attribute=true
    SpriteSize=&Multiply_Vector2D_ByFloat(Float=1.5, Vector2D=$Particles.SpriteSize)
    Write to Module Namespace=true
   ) {module=FN_Sprite_Attributes001}
   &FN_Light_Attributes(
    Color=$Particles.Color
    Create Color Attribute=true
    Create Exponent Attribute=true
    Create Radius Attribute=true
    Enable Particle Light Enable=true
    Exponent=2
    Light Enabled=true
    Radius=750
    Scale Color=&FloatFromCurve(
     FloatCurve=DataInterface() {propertyValues {Curve {keys [
         {arriveTangent=3.1076, interpMode=RCIM_Cubic, leaveTangent=3.1076, tangentMode=RCTM_User, time=0, value=0}
         {interpMode=RCIM_Cubic, tangentMode=RCTM_User, time=1, value=1}
        ]}}}
     CurveIndex=&Multiply_Float(A=$System.ActivationAlpha, B=$System.FadeOut)
     Scale Curve=10
    )
   )
   &DynamicMaterialParameters(
    Index 3 Param 1=1
    Index 3 Param 4=&Multiply_Float(A=$Engine.System.Age, B=-2)
    Module.Index3Float0Write=true
    Module.Index3Float1Write=true
    Module.Index3Float2Write=true
    Module.Index3Float3Write=true
    Write Parameter Index 0=::NewEnumerator0
    Write Parameter Index 3=::NewEnumerator1
   )
   &FN_Mesh_Attributes(
    Color=&MultiplyLinearColor(
     Linear Color=&ColorFromCurve(
      ColorCurve=DataInterface() {propertyValues {
        AlphaCurve {keys [<at 0.0077 0>,<at 1 1 RCIM_Cubic>]}
        BlueCurve {keys [<at 0 0.9942 RCIM_Cubic>,<at 0.9711 0.3946 RCIM_Cubic>]}
        GreenCurve {keys [<at 0 0.1305 RCIM_Cubic>,<at 0.9711 0.1523 RCIM_Cubic>]}
        RedCurve {keys [<at 0 1 RCIM_Cubic>,<at 0.9711 1 RCIM_Cubic>]}
       }}
      CurveIndex=$System.ActivationAlpha
     )
     Scale Factor=&MakeFloat4FromVectorAndFloat(
      "Vector (RGB)"=&VectorFromFloat(Value=&FloatFromCurve(
        FloatCurve=DataInterface() {propertyValues {Curve {keys [<at 0 0 RCIM_Cubic>, {interpMode=RCIM_Cubic, tangentMode=RCTM_Break, time=1, value=1}]}}}
        CurveIndex=$System.ActivationAlpha
        Scale Curve=2
       ))
      "Float (Alpha)"=&FloatFromCurve(
       FloatCurve=DataInterface() {propertyValues {Curve {keys [<at 0 0>,<at 1 1 RCIM_Cubic>]}}}
       CurveIndex=&Multiply_Float(A=$System.ActivationAlpha, B=$System.FadeOut)
      )
     )
    )
    Create Color Attribute=true
    Create Scale Attribute=true
    Scale=&VectorFromFloat(Value=&FloatFromCurve(
      FloatCurve=DataInterface() {propertyValues {Curve {keys [
          {arriveTangent=2.7841, interpMode=RCIM_Cubic, leaveTangent=2.7841, tangentMode=RCTM_User, time=0, value=0}
          {arriveTangent=0.1404, interpMode=RCIM_Cubic, leaveTangent=0.1404, tangentMode=RCTM_User, time=1, value=1}
         ]}}}
      CurveIndex=$System.ActivationAlpha
      Scale Curve=35
     ))
    Write to Module Namespace=true
   ) {module=FN_Mesh_Attributes001}
   /SetVariables(Particles.EnergyMeshDynMat:Vector4f=&MakeVector4(
     X=&FloatFromCurve(FloatCurve=DataInterface() {propertyValues {Curve {keys [{arriveTangent=-2.1015, interpMode=RCIM_Cubic, leaveTangent=-2.1015, tangentMode=RCTM_User, time=0, value=0.5}, <at 1 -0.15>]}}})
     Y=&FloatFromCurve(FloatCurve=DataInterface() {propertyValues {Curve {keys [{arriveTangent=-0.3464, interpMode=RCIM_Cubic, leaveTangent=-0.3464, tangentMode=RCTM_User, time=0, value=1}, <at 1 0.9>]}}})
     Z=&FloatFromCurve(FloatCurve=DataInterface() {propertyValues {Curve {keys [
          {interpMode=RCIM_Cubic, tangentMode=RCTM_Break, time=0, value=0}
          {arriveTangent=0.4478, interpMode=RCIM_Cubic, leaveTangent=0.4478, tangentMode=RCTM_User, time=0.15, value=1}
          {interpMode=RCIM_Cubic, tangentMode=RCTM_Break, time=1, value=1.2}
         ]}}})
     W=0
    )) {isSetParameters=true}
   &FN_Mesh_Attributes(
    Color=&MultiplyLinearColor(
     Linear Color=&ColorFromCurve(
      ColorCurve=DataInterface() {propertyValues {
        AlphaCurve {keys [<at 1 1 RCIM_Cubic>]}
        BlueCurve {keys [<at 0 0.9942>, {interpMode=RCIM_Cubic, leaveTangent=-0.5226, time=1, value=1.9533}]}
        GreenCurve {keys [<at 0 0.1305>, {interpMode=RCIM_Cubic, leaveTangent=0.126, time=1, value=0.4931}]}
        RedCurve {keys [<at 0 1>,<at 1 5 RCIM_Cubic>]}
       }}
      CurveIndex=$System.ActivationAlpha
     )
     Scale Factor=&MakeFloat4FromVectorAndFloat(
      "Vector (RGB)"=&VectorFromFloat(Value=&FloatFromCurve(
        FloatCurve=DataInterface() {propertyValues {Curve {keys [
            {interpMode=RCIM_Cubic, leaveTangent=1.9522, tangentMode=RCTM_Break, time=0, value=0}
            {arriveTangent=0.1224, interpMode=RCIM_Cubic, tangentMode=RCTM_Break, time=1, value=1}
           ]}}}
        CurveIndex=$System.ActivationAlpha
        Scale Curve=10
       ))
      "Float (Alpha)"=&FloatFromCurve(
       FloatCurve=DataInterface() {propertyValues {Curve {keys [<at 0 0 RCIM_Cubic>,<at 1 1>]}}}
       CurveIndex=&Multiply_Float(A=$System.ActivationAlpha, B=$System.FadeOut)
      )
     )
    )
    Create Color Attribute=true
    Create Scale Attribute=true
    Scale=&VectorFromFloat(Value=&FloatFromCurve(
      FloatCurve=DataInterface() {propertyValues {Curve {keys [
          {arriveTangent=2.5597, interpMode=RCIM_Cubic, leaveTangent=2.5597, tangentMode=RCTM_User, time=0, value=0}
          {interpMode=RCIM_Cubic, tangentMode=RCTM_User, time=1, value=1}
         ]}}}
      CurveIndex=$System.ActivationAlpha
      Scale Curve=80
     ))
    Write to Module Namespace=true
   ) {module=FN_Mesh_Attributes002}
   &FN_Mesh_Attributes(
    Color=&MultiplyLinearColor(
     Linear Color=<rgba 1 1 1 1>
     Scale Factor=&MakeFloat4FromVectorAndFloat("Vector (RGB)"=<1 1 1>, "Float (Alpha)"=$System.FadeOut)
    )
    Create Color Attribute=true
    Create Orientation Attribute=true
    Create Scale Attribute=true
    Scale=&VectorFromFloat(Value=&Multiply_Float(A=25, B=$System.FadeOut))
    Write to Module Namespace=true
   )
  ]
  rendererClasses [
   &NiagaraSpriteRendererProperties
   &NiagaraLightRendererProperties
   &NiagaraMeshRendererProperties
  ]
  renderers [
   {rendererClass=&NiagaraSpriteRendererProperties}
   {rendererClass=&NiagaraSpriteRendererProperties, rendererIndex=1}
   {rendererClass=&NiagaraSpriteRendererProperties, rendererIndex=2}
   {rendererClass=&NiagaraLightRendererProperties, rendererIndex=3}
   {rendererClass=&NiagaraMeshRendererProperties, rendererIndex=4}
   {rendererClass=&NiagaraMeshRendererProperties, rendererIndex=5}
   {rendererClass=&NiagaraMeshRendererProperties, rendererIndex=6}
  ]
 }
 NE_Lightnings {
  emitterUpdateScript [
   &EmitterState()
   &SpawnRate(
    SpawnRate=&Multiply_Float(A=&Lerp_Float(A=0, B=15, Alpha=$System.ActivationAlpha), B=$System.FadeOut)
    Use Spawn Probability=false
   )
  ]
  particleSpawnScript [
   &InitializeParticle(
    Color=&MultiplyLinearColor(
     Linear Color=&RandomRangeLinearColor(Minimum=<rgba 1 1 1 0.3>, Maximum=<rgba 1.8 1.8 1.8 1>)
     Scale Factor=&MakeVector4_FromFloat(Float=0.8)
    )
    Lifetime=&RandomRangeFloat(Minimum=0.4, Maximum=0.5)
    Mesh Scale Mode=::NewEnumerator3
    Mesh Uniform Scale=3
    Sprite Size Mode=::NewEnumerator3
    Uniform Sprite Size=20
    UsePositionOffset=true
    Write Material Random=true
   )
   &InitialMeshOrientation(Use Rotation Vector=true)
   &ShapeLocation(
    Cylinder Distribution=::NewEnumerator2
    Cylinder Height=5
    Cylinder Radius=1200
    MaskSpawnGroup=false
    Rotation Mode=::NewEnumerator2
    Shape Primitive=::NewEnumerator1
    U Position=&Lerp_Float(A=0, B=1, Alpha=&Multiply_Float(A=$Engine.System.Age, B=1.7452))
    UseOwnerScale=true
    Yaw / Pitch / Roll=<0 0.25 0>
   )
  ]
  particleUpdateScript [
   &ParticleState()
   &ScaleMeshSize(Scale Factor=&RandomRangeVector(Minimum=<1 1 2.2>, Maximum=<1.3 1.3 2.2>)) {module=ScaleMeshSize001}
   &DynamicMaterialParameters(
    Index 0 Param 2=&RandomRangeFloat(Minimum=0, Maximum=0)
    Index 0 Param 4=13
    Index 1 Param 1=&FloatFromCurve(FloatCurve=DataInterface() {propertyValues {Curve {keys [<at 0 0.5>,<at 1 1>]}}})
    Index 1 Param 2=2
    Index 1 Param 3=&FloatFromCurve(FloatCurve=DataInterface() {propertyValues {Curve {keys [{interpMode=RCIM_Cubic, tangentMode=RCTM_Break, time=0.5, value=-0.5}, <at 1 0.1>]}}})
    Index 2 Param 1=&Multiply_Float(
     A=&FloatFromCurve(
      FloatCurve=DataInterface() {propertyValues {Curve {keys [{interpMode=RCIM_Cubic, leaveTangent=8.1085, tangentMode=RCTM_Break, time=0.8568, value=1}, {tangentMode=RCTM_User, time=1, value=0}]}}}
      Scale Curve=0.2
     )
     B=&Divide_Float(
      A=&MakeFloatFromVector(Vector=$Particles.Scale, Channel=::NewEnumerator0)
      B=&RandomRangeFloat(Minimum=10, Maximum=15)
     )
    )
    Index 2 Param 2=&FloatFromCurve(
     FloatCurve=DataInterface() {propertyValues {Curve {keys [
         {interpMode=RCIM_Cubic, tangentMode=RCTM_Break, time=0.7, value=0}
         {arriveTangent=-0.0859, interpMode=RCIM_Cubic, tangentMode=RCTM_Break, time=1, value=-1}
        ]}}}
     Scale Curve=50
    )
    Index 2 Param 3=&Waveform(
     Global Input Position=&RandomRangeFloat(Minimum=0, Maximum=15)
     Global Phase=$Particles.Age
     "Waveform [1]"=::NewEnumerator12
     "[1] Phase Scale"=2
     "[1] Frequency"=1
     "[1] Amplitude Min/Max"=<0 1>
    )
    Index 2 Param 4=25
    Index 3 Param 1=&FloatFromCurve(
     FloatCurve=DataInterface() {propertyValues {Curve {keys [
         {interpMode=RCIM_Cubic, leaveTangent=-0.0025, tangentMode=RCTM_Break, time=0, value=0}
         {arriveTangent=1.8528, interpMode=RCIM_Cubic, tangentMode=RCTM_Break, time=1, value=1}
        ]}}}
     Scale Curve=&RandomRangeFloat(Minimum=0, Maximum=0)
    )
    Index 3 Param 2=&FloatFromCurve(
     FloatCurve=DataInterface() {propertyValues {Curve {keys [
         {interpMode=RCIM_Cubic, leaveTangent=-0.0025, tangentMode=RCTM_Break, time=0, value=0.5}
         {arriveTangent=-0.0316, interpMode=RCIM_Cubic, tangentMode=RCTM_Break, time=1, value=-0.2}
        ]}}}
     Scale Curve=0.1
    )
    Index 3 Param 3=&FloatFromCurve(FloatCurve=DataInterface() {propertyValues {Curve {keys [
         {interpMode=RCIM_Cubic, leaveTangent=1.1034, tangentMode=RCTM_Break, time=0, value=-1}
         {arriveTangent=2.2399, tangentMode=RCTM_User, time=0.7, value=0.125}
        ]}}})
    Index 3 Param 4=&FloatFromCurve(FloatCurve=DataInterface() {propertyValues {Curve {keys [
         {interpMode=RCIM_Cubic, leaveTangent=3.0601, tangentMode=RCTM_Break, time=0.5, value=-0.5}
         {arriveTangent=3.2053, tangentMode=RCTM_User, time=1, value=1}
        ]}}})
    Module.Index0Param0Write=true
    Module.Index0Param1Write=true
    Module.Index0Param2Write=true
    Module.Index0Param3Write=true
    Module.Index1Param0Write=true
    Module.Index1Param1Write=true
    Module.Index1Param2Write=true
    Module.Index1Param3Write=true
    Module.Index2Param0Write=true
    Module.Index2Param1Write=true
    Module.Index2Param2Write=true
    Module.Index2Param3Write=true
    Module.Index3Param0Write=true
    Module.Index3Param1Write=true
    Module.Index3Param2Write=true
    Module.Index3Param3Write=true
    Write Parameter Index 1=true
    Write Parameter Index 2=true
    Write Parameter Index 3=true
   )
   &SolveForcesAndVelocity(Clamp Velocity=false, Limit Acceleration=false)
   &Color(
    Color=&MultiplyLinearColor(
     Linear Color=&ColorFromCurve(ColorCurve=DataInterface() {propertyValues {
        AlphaCurve {keys [<at 0 1 RCIM_Cubic>,<at 0.7884 1>,<at 1 0 RCIM_Cubic>]}
        BlueCurve {keys [<at 0.2225 5>, {interpMode=RCIM_Cubic, leaveTangent=-19.7604, time=0.4906, value=1.5}]}
        GreenCurve {keys [<at 0.2225 5>, {interpMode=RCIM_Cubic, leaveTangent=-19.7604, time=0.4906, value=1.5}]}
        RedCurve {keys [<at 0.2225 5>, {interpMode=RCIM_Cubic, leaveTangent=-19.7604, time=0.4906, value=1.5}]}
       }})
     Scale Factor=&MakeVector4_FromLinearColor(LinearColor=<rgba 1 0.1302 0.4126 1>)
    )
    Scale Alpha=&FloatFromCurve(FloatCurve=DataInterface() {propertyValues {Curve {keys [
         {interpMode=RCIM_Cubic, leaveTangent=2.5294, tangentMode=RCTM_Break, time=0, value=0}
         {arriveTangent=-0.2166, interpMode=RCIM_Cubic, tangentMode=RCTM_Break, time=0.4, value=1}
        ]}}})
   )
   /SetVariables(Particles.GlowDecalSize:Vector2f=&Multiply_Vector2D_ByFloat(Float=150, Vector2D=<1 1>)) {isSetParameters=true}
   /SetVariables(Particles.DecalDynMat:Vector4f=<5 0 0 0>) {isSetParameters=true}
   /SetVariables(Particles.DecalColor:LinearColor=&MultiplyLinearColor(Linear Color=<rgba 2 2 2 1>, Scale Factor=&MakeVector4_FromLinearColor(LinearColor=$Particles.Color))) {isSetParameters=true}
   &CameraOffset(Camera Offset Amount=50)
  ]
  rendererClasses [&NiagaraMeshRendererProperties, &NiagaraSpriteRendererProperties]
  renderers [
   {rendererClass=&NiagaraMeshRendererProperties, rendererIndex=0}
   {rendererClass=&NiagaraSpriteRendererProperties, rendererIndex=1}
   {rendererClass=&NiagaraMeshRendererProperties, rendererIndex=2}
  ]
 }
 SmokGround {
  simTarget=GPUComputeSim
  emitterUpdateScript [
   &EmitterState()
   &SpawnRate(
    SpawnRate=&Multiply_Float(A=&Lerp_Float(A=1.5, B=2.5, Alpha=$System.ActivationAlpha), B=$System.FadeOut)
    Use Spawn Probability=false
   )
  ]
  particleSpawnScript [
   &InitializeParticle(
    Color=<rgba 1 0 0.3333 0.8>
    Lifetime=&Lerp_Float(
     A=&RandomRangeFloat(Minimum=8.5, Maximum=17.5)
     B=&RandomRangeFloat(Minimum=4.5, Maximum=9.5)
     Alpha=$System.ActivationAlpha
    )
    Position Offset=<0 0 -869.0536>
    Sprite Size Mode=::NewEnumerator3
    Sprite UV Mode=::NewEnumerator3
    Uniform Sprite Size=&RandomRangeFloat(Minimum=1, Maximum=2)
    UsePositionOffset=true
    Write Material Random=true
   )
   &ShapeLocation(
    Cylinder Radius=1351.9327
    Hemicircle Internal Cap=false
    Hemicircle X=true
    MaskSpawnGroup=false
    Override Seed=false
    Shape Primitive=::NewEnumerator1
    Surface Only=true
    Use Endcaps In Surface Only Mode=false
    UseOwnerScale=true
   )
   /SetVariables(Particles.SpriteRotation:NiagaraFloat=&RandomRangeFloat(Minimum=-0.2293, Maximum=0.3106)) {isSetParameters=true, module=SetVariables_#4}
   &AddVelocity(Velocity=$StackContext.ShapeLocation.ShapeVector, Velocity Speed Scale=0.1)
   /SetVariables(Particles.SpriteSize=&Multiply_Vector2D_ByFloat(
     Float=&Lerp_Float(A=107.586, B=190.6381, Alpha=&MakeFloatFromInt(Int=$Particles.SpawnGroup))
     Vector2D=$Particles.SpriteSize
    )) {isSetParameters=true, module=SetVariables_#5}
   &ApplyInitialForces(Positional Force Warmup Time=5)
  ]
  particleUpdateScript [
   &Drag(Drag=0.1, Use Linear Drag=true, Use Rotational Drag=true)
   &VortexForce(
    Origin Pull Amount=76.4139
    UseOriginOffset=false
    Vortex Axis=<1 0 0>
    Vortex Force Amount=&Lerp_Float(
     A=&RandomRangeFloat(Minimum=93.3802, Maximum=178.6482)
     B=0
     Alpha=&MakeFloatFromInt(Int=$Particles.SpawnGroup)
    )
   ) {enabled=false}
   &GravityForce(Gravity=&Lerp_Vector(A=<0 0 -85.0702>, B=<0 0 -226.7489>, Alpha=&MakeFloatFromInt(Int=$Particles.SpawnGroup))) {enabled=false}
   &AccelerationForce(Acceleration=<200 0 0>, Coordinate Space=::Local) {enabled=false}
   &SolveForcesAndVelocity(Clamp Velocity=false, Limit Acceleration=false)
   &ParticleState()
   &ScaleSpriteSize(
    Uniform Curve Scale=5
    Uniform Curve Sprite Scale=DataInterface() {propertyValues {Curve {keys [{arriveTangent=1.2245, interpMode=RCIM_Cubic, leaveTangent=1.2245, tangentMode=RCTM_User, time=0, value=1.5}, <at 1 2 RCIM_Cubic>]}}}
   )
   /SetVariables(Particles.SubImageIndex=&FloatFromCurve(FloatCurve=DataInterface() {propertyValues {Curve {keys [<at 0 0>,<at 1 1>]}}}, Scale Curve=16)) {isSetParameters=true, module=SetVariables_#6}
   &ScaleColor(
    Scale Alpha=&FloatFromCurve(
     FloatCurve=DataInterface() {propertyValues {Curve {keys [
         <at 0 0 RCIM_Cubic>
         {arriveTangent=1, interpMode=RCIM_Cubic, leaveTangent=1, time=0.2, value=0.775}
         {arriveTangent=-0.9688, interpMode=RCIM_Cubic, leaveTangent=-0.9688, time=0.7, value=0.7}
         <at 1 0 RCIM_Cubic>
        ]}}}
     Scale Curve=0.5
    )
    Scale RGB=&VectorFromFloat(Value=&FloatFromCurve(FloatCurve=DataInterface() {propertyValues {Curve {keys [{arriveTangent=-0.5678, interpMode=RCIM_Cubic, leaveTangent=-0.5678, tangentMode=RCTM_User, time=0, value=0.1}, <at 0.5 0 RCIM_Cubic>]}}}))
    ScaleA=true
    ScaleRGB=true
   )
   &DynamicMaterialParameters(
    Index 0 Param 4=&FloatFromCurve(FloatCurve=DataInterface() {propertyValues {Curve {keys [<at 0 1>,<at 1 0>]}}}, Scale Curve=2)
    Module.Index0Param0Write=true
    Module.Index0Param1Write=true
    Module.Index0Param2Write=true
    Module.Index0Param3Write=true
   )
   &CameraOffset(Camera Offset Amount=&FloatFromCurve(FloatCurve=DataInterface() {propertyValues {Curve {keys [<at 0 -50>,<at 1 -10>]}}}))
   &self:MatchScale() {module=MatchScale}
  ]
  rendererClasses [&NiagaraSpriteRendererProperties]
  renderers [{rendererClass=&NiagaraSpriteRendererProperties}]
 }
 Smoke {
  simTarget=GPUComputeSim
  emitterUpdateScript [
   &EmitterState()
   &SpawnRate(
    SpawnRate=&Multiply_Float(A=&Lerp_Float(A=5, B=25, Alpha=$System.ActivationAlpha), B=$System.FadeOut)
    Use Spawn Probability=false
   )
  ]
  particleSpawnScript [
   &InitializeParticle(
    Color=<rgba 1 0 0.3333 1>
    Lifetime=&Lerp_Float(
     A=&RandomRangeFloat(Minimum=3.1968, Maximum=7.0176)
     B=&RandomRangeFloat(Minimum=1, Maximum=2)
     Alpha=$System.ActivationAlpha
    )
    Sprite Size Mode=::NewEnumerator3
    Sprite UV Mode=::NewEnumerator3
    Uniform Sprite Size=&RandomRangeFloat(Minimum=0.5, Maximum=1)
    UsePositionOffset=false
    Write Material Random=true
   )
   &ShapeLocation(
    MaskSpawnGroup=true
    Override Seed=false
    Ring Radius=970.7466
    Rotation Mode=::NewEnumerator2
    Shape Primitive=::NewEnumerator4
    UseOwnerScale=true
    Yaw / Pitch / Roll=<0 0.25 0>
   )
   &ShapeLocation(
    MaskSpawnGroup=true
    Override Seed=false
    Rotation Mode=::NewEnumerator2
    Shape Origin=&AddVectorToPosition(
     Position=$StackContext.Position
     Vector=&Multiply_Vector(
      A=<207.3846 550.5953 -570.5737>
      B=&MakeVector(
       X=1
       Y=&UniformAOrBFloat(A=-1, Evaluation Type=::NewEnumerator0, Recalculate Random Each Loop=true, Override Seed=false)
       Z=1
      )
     )
    )
    Spawn Group Mask=1
    Sphere Radius=148.652
    UseOwnerScale=true
    Yaw / Pitch / Roll=<0.25 0.25 0>
   ) {module=ShapeLocation001}
   /SetVariables(Particles.SpriteRotation:NiagaraFloat=&RandomRangeFloat(Minimum=-0.2293, Maximum=0.3106)) {isSetParameters=true, module=SetVariables_#4}
   &AddVelocity(Velocity=$StackContext.ShapeLocation.ShapeVector, Velocity Speed Scale=0)
   /SetVariables(Particles.SpriteSize=&Multiply_Vector2D_ByFloat(
     Float=&Lerp_Float(A=107.586, B=190.6381, Alpha=&MakeFloatFromInt(Int=$Particles.SpawnGroup))
     Vector2D=$Particles.SpriteSize
    )) {isSetParameters=true, module=SetVariables_#5}
   &ApplyInitialForces()
  ]
  particleUpdateScript [
   &Drag(Drag=2, Use Linear Drag=true, Use Rotational Drag=true)
   &VortexForce(
    Origin Pull Amount=76.4139
    UseOriginOffset=false
    Vortex Axis=<1 0 0>
    Vortex Force Amount=&Lerp_Float(
     A=&RandomRangeFloat(Minimum=93.3802, Maximum=178.6482)
     B=0
     Alpha=&MakeFloatFromInt(Int=$Particles.SpawnGroup)
    )
   )
   &GravityForce(Gravity=&Lerp_Vector(A=<0 0 -85.0702>, B=<0 0 -226.7489>, Alpha=&MakeFloatFromInt(Int=$Particles.SpawnGroup)))
   &AccelerationForce(Acceleration=<300 0 0>, Coordinate Space=::Local)
   &SolveForcesAndVelocity(Clamp Velocity=false, Limit Acceleration=false)
   &ParticleState()
   &ScaleSpriteSize(
    Uniform Curve Scale=5
    Uniform Curve Sprite Scale=DataInterface() {propertyValues {Curve {keys [{arriveTangent=1.2245, interpMode=RCIM_Cubic, leaveTangent=1.2245, tangentMode=RCTM_User, time=0, value=1.5}, <at 1 2 RCIM_Cubic>]}}}
   )
   /SetVariables(Particles.SubImageIndex=&FloatFromCurve(FloatCurve=DataInterface() {propertyValues {Curve {keys [<at 0 0>,<at 1 1>]}}}, Scale Curve=16)) {isSetParameters=true, module=SetVariables_#6}
   &ScaleColor(
    Scale Alpha=&FloatFromCurve(FloatCurve=DataInterface() {propertyValues {Curve {keys [
         <at 0 0 RCIM_Cubic>
         {arriveTangent=0.8929, interpMode=RCIM_Cubic, leaveTangent=0.8929, time=0.2, value=1}
         {arriveTangent=-1.25, interpMode=RCIM_Cubic, leaveTangent=-1.25, time=0.7, value=0.625}
         <at 1 0 RCIM_Cubic>
        ]}}})
    Scale RGB=&VectorFromFloat(Value=&FloatFromCurve(
      FloatCurve=DataInterface() {propertyValues {Curve {keys [{arriveTangent=-5.336, interpMode=RCIM_Cubic, leaveTangent=-5.336, tangentMode=RCTM_User, time=0, value=1}, <at 0.5 0.1>]}}}
      Scale Curve=&Lerp_Float(A=1, B=3, Alpha=$System.ActivationAlpha)
     ))
    ScaleA=true
    ScaleRGB=true
   )
   &DynamicMaterialParameters(
    Index 0 Param 4=&FloatFromCurve(FloatCurve=DataInterface() {propertyValues {Curve {keys [<at 0 1>,<at 1 0>]}}}, Scale Curve=2)
    Module.Index0Param0Write=true
    Module.Index0Param1Write=true
    Module.Index0Param2Write=true
    Module.Index0Param3Write=true
   )
   &CameraOffset(Camera Offset Amount=&FloatFromCurve(FloatCurve=DataInterface() {propertyValues {Curve {keys [<at 0 25>,<at 1 75>]}}}))
   &self:MatchScale() {module=MatchScale}
  ]
  rendererClasses [&NiagaraSpriteRendererProperties]
  renderers [{rendererClass=&NiagaraSpriteRendererProperties}]
 }
 SparksLayer1 {
  simTarget=GPUComputeSim
  emitterUpdateScript [
   &EmitterState(Life Cycle Mode=::NewEnumerator1, Loop Duration=1, UseLoopDelay=false)
   &SpawnRate(SpawnRate=&Multiply_Float(A=20000, B=$System.FadeOut), Use Spawn Probability=false)
  ]
  particleSpawnScript [
   &InitializeParticle(
    Lifetime=&FloatFromCurve(
     FloatCurve=DataInterface() {propertyValues {Curve {keys [
         {interpMode=RCIM_Cubic, tangentMode=RCTM_User, time=0, value=3}
         {interpMode=RCIM_Cubic, tangentMode=RCTM_User, time=0.5, value=7}
         {tangentMode=RCTM_User, time=1, value=3}
        ]}}}
     CurveIndex=&Frac_Float(Float=&Multiply_Float(A=$Engine.Time, B=&Lerp_Float(A=1, B=2, Alpha=$System.ActivationAlpha)))
     Scale Curve=&Lerp_Float(A=1, B=0.45, Alpha=$System.ActivationAlpha)
    )
    Mass Mode=::NewEnumerator1
    Sprite Size=&Multiply_Vector2(
     A=&Vector2DFromFloat(Value=&UniformRangedFloat(Minimum=0.4, Maximum=0.6))
     B=<2 6>
    )
    Sprite Size Mode=::NewEnumerator4
    UsePositionOffset=false
    Write Material Random=true
   ) {module=InitializeParticle001}
   &CylinderLocation(
    Cylinder Height=1
    Cylinder Radius=$System.SparkRadius
    Hemisphere X=true
    Hemisphere Y=true
    MaskSpawnGroup=false
    Module.Enable Lathe Profile=false
    Module.Override Randomness=false
    Module.Surface Only=true
    Orientation Axis:ENiagaraOrientationAxis=::XAxis
    Use Endcaps In Surface Only Mode=false
   )
   &self:InitialPosCHange(Angle:NiagaraFloat=&Multiply_Float(A=180, B=$Engine.Time)) {module=InitialPosCHange}
   /SetVariables(Transient.Position2:NiagaraPosition=$Particles.Position) {isSetParameters=true}
   &CylinderLocation(
    Cylinder Height=1
    Cylinder Origin:Vector3f=/Niagara/DynamicInputs/Helpers/SimulationPosition()
    Cylinder Radius=$System.SparkRadius
    Lathe Profile=DataInterface() {propertyValues {Curve {keys [<at 0 0.451 RCIM_Cubic>, {arriveTangent=0.049, interpMode=RCIM_Cubic, leaveTangent=0.049, time=0.5, value=1}, <at 1 0.5 RCIM_Cubic>]}}}
    MaskSpawnGroup=false
    Module.Enable Lathe Profile=false
    Module.Override Randomness=false
    Module.Surface Only=true
    Orientation Axis:ENiagaraOrientationAxis=::XAxis
    Use Endcaps In Surface Only Mode=false
   ) {module=CylinderLocation001}
   &self:ChoosePos(Lerp=0.3) {module=ChoosePos}
   /Niagara/Modules/Update/Forces/CurlNoiseForce(Noise Frequency=0.005, Noise Quality / Cost=::NewEnumerator3, Noise Strength=50, Pan Noise=true, Pan Noise Field=<1 0 0>) {module=CurlNoiseForce001}
   &PointForce(
    Force Strength=&Lerp_Float(A=-250, B=-1250, Alpha=$System.ActivationAlpha)
    Module.Use Falloff=false
    Module.UseLinearForce=false
   ) {module=PointForce001}
   &ApplyInitialForces()
   /SetVariables(Particles.Position=&Add_Vector(A=Linked() {linkedVariable {name=Particles.Position, type=NiagaraPosition}}, B=$Transient.PhysicsForce)) {isSetParameters=true}
  ]
  particleUpdateScript [
   &ParticleState()
   &CurlNoiseForce(Noise Frequency=0.0015, Noise Quality / Cost=::NewEnumerator3, Noise Strength=7.5, Pan Noise=true, Pan Noise Field=<0.5 0 0>)
   &self:ResampleFroce(Intensity=3) {module=ResampleFroce}
   &VortexForce(Influence Falloff Exponent=1.5, Influence Falloff Radius=1500, Use Influence Falloff=true, Vortex Axis=<1 0 0>, Vortex Force Amount=800)
   &PointForce(Force Falloff Distance=1000, Force Strength=400, Module.Use Falloff=true, Module.UseLinearForce=false)
   &Drag(
    Drag=&FloatFromCurve(
     FloatCurve=DataInterface() {propertyValues {Curve {keys [
         <at -0.0101 2.7154>
         {arriveTangent=-0.3026, interpMode=RCIM_Cubic, leaveTangent=-0.3026, tangentMode=RCTM_User, time=0.0352, value=1.1374}
         {arriveTangent=1.8448, interpMode=RCIM_Cubic, leaveTangent=1.8448, time=0.5183, value=1.5087}
         {arriveTangent=3.0746, interpMode=RCIM_Cubic, leaveTangent=3.0746, tangentMode=RCTM_User, time=0.8151, value=2.5761}
        ]}}}
     CurveIndex=&Divide_Float(A=$Particles.Age, B=6.5)
    )
    Use Linear Drag=true
    Use Rotational Drag=false
   )
   &SolveForcesAndVelocity(Clamp Velocity=false, Limit Acceleration=false)
   &ScaleColor(
    Scale Mode=::NewEnumerator1
    Scale RGBA=&Vector4FromCurve(Vector4Curve=DataInterface() {propertyValues {
       AlphaCurve {keys [
         <at 0.0125 0 RCIM_Cubic>
         {arriveTangent=0.5858, interpMode=RCIM_Cubic, leaveTangent=0.5858, time=0.0846, value=0.5}
         {arriveTangent=-0.5462, interpMode=RCIM_Cubic, leaveTangent=-0.5462, time=0.8661, value=0.5}
         <at 1 0 RCIM_Cubic>
        ]}
       BlueCurve {keys [<at 0.0887 2.7858>,<at 0.2437 0.2048>,<at 0.6292 0>]}
       GreenCurve {keys [<at 0.0887 0.6772>,<at 0.2437 0.0263>,<at 0.6292 0>]}
       RedCurve {keys [<at 0.0887 10>,<at 0.2437 1>,<at 0.6292 0>]}
      }})
    ScaleRGBA=true
   )
   &ScaleSpriteSize(Scale Sprite Size Mode=::NewEnumerator0, Uniform Scale Factor=6) {enabled=false}
   &ScaleSpriteSizeBySpeed(Max Scale Factor=<6 10>, Min Scale Factor=<6 3>, Velocity Threshold=300)
   &self:MatchScale() {module=MatchScale}
   &CameraOffset(Camera Offset Amount=&FloatFromCurve(
     FloatCurve=DataInterface() {propertyValues {Curve {keys [{tangentMode=RCTM_User, time=0, value=0}, {tangentMode=RCTM_User, time=1, value=2}]}}}
     Scale Curve=25
    ))
  ]
  rendererClasses [&NiagaraSpriteRendererProperties]
  renderers [{rendererClass=&NiagaraSpriteRendererProperties}]
 }
}
summary {
 systemName=NE_GiftedTraveler_GateVFX
 emitters [
  {emitterName=SparksLayer1, simTarget=GPUComputeSim, rendererClasses [&NiagaraSpriteRendererProperties]}
  {emitterName=Center, rendererClasses [&NiagaraSpriteRendererProperties]}
  {emitterName=Smoke, simTarget=GPUComputeSim, rendererClasses [&NiagaraSpriteRendererProperties]}
  {emitterName=MeshRibbon, rendererClasses [&NiagaraMeshRendererProperties]}
  {emitterName=SmokGround, simTarget=GPUComputeSim, rendererClasses [&NiagaraSpriteRendererProperties]}
  {emitterName=NE_Lightnings, rendererClasses [&NiagaraMeshRendererProperties, &NiagaraSpriteRendererProperties]}
  {
   emitterName=NE_GlowFlare
   rendererClasses [
    &NiagaraSpriteRendererProperties
    &NiagaraLightRendererProperties
    &NiagaraMeshRendererProperties
   ]
  }
 ]
 userVariables [
  User.ChaosDestructionData {
   type=NiagaraDataInterfaceChaosDestruction
   description=""
   defaultValue {
    struct=NiagaraExt_VariableValue_DataInterface
    value {
     dataInterface=&self:NiagaraDataInterfaceChaosDestruction_0
     dataInterfaceClass=/Script/ChaosNiagara.NiagaraDataInterfaceChaosDestruction
    }
   }
  }
  User.FadeOut {type=NiagaraBool}
  User.IsHit {type=NiagaraBool}
 ]
}
